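# Physics asset for the back shirt flaps of the F_MED_Believer player cosmetic body, with capsule bodies for three flap chains and two spine bones.
# Each flap bone is constrained to its base bone as a narrow hinge, with one limited swing axis and the other swing and twist locked.
# Flap bases A and B attach to spine_05 and C to spine_04, with angular motion free on these joints and between the spine bones.
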PhysicsAsset /BRCosmetics/Characters/Player/Female/Medium/Bodies/F_MED_Believer/Meshes/F_MED_Believer_ShirtsBack_Physics
bodies {
 deault_backFlap_a [{shapeType=Capsule, length=3, radius=2, center=<4 0 0>, rotation=<rot 90 0 0>}]
 deault_backFlap_a_base [{shapeType=Capsule, length=1, radius=1}]
 default_backFlap_b [{shapeType=Capsule, length=3, radius=2, center=<4 0 0>, rotation=<rot 90 0 0>}]
 default_backFlap_b_base [{shapeType=Capsule, length=1, radius=1}]
 default_backFlap_c [{shapeType=Capsule, length=3, radius=2, center=<4 0 0>, rotation=<rot 90 0 0>}]
 default_backFlap_c_base [{shapeType=Capsule, length=1, radius=1}]
 spine_04 [{shapeType=Capsule, length=3.6359, radius=9.5052, center=<-2.4339 2.9787 -0.0561>, rotation=<rot -0.1889 -90 -0.1263>}]
 spine_05 [{shapeType=Capsule, length=7.75, radius=7.75, center=<0 1.4771 0>}]
}
constraints [
 {
  bone1Name=deault_backFlap_a
  bone2Name=deault_backFlap_a_base
  swing1LimitDegrees=20
  swing2LimitDegrees=5
  swing2Motion=Locked
  twistLimitDegrees=45
  twistMotion=Locked
 }
 {
  bone1Name=default_backFlap_b
  bone2Name=default_backFlap_b_base
  swing1LimitDegrees=20
  swing2LimitDegrees=5
  swing2Motion=Locked
  twistLimitDegrees=45
  twistMotion=Locked
 }
 {
  bone1Name=default_backFlap_c
  bone2Name=default_backFlap_c_base
  swing1LimitDegrees=20
  swing2LimitDegrees=5
  swing2Motion=Locked
  twistLimitDegrees=45
  twistMotion=Locked
 }
 {
  bone1Name=deault_backFlap_a_base
  bone2Name=spine_05
  swing1LimitDegrees=45
  swing1Motion=Free
  swing2LimitDegrees=45
  swing2Motion=Free
  twistLimitDegrees=45
  twistMotion=Free
 }
 {
  bone1Name=default_backFlap_b_base
  bone2Name=spine_05
  swing1LimitDegrees=45
  swing1Motion=Free
  swing2LimitDegrees=45
  swing2Motion=Free
  twistLimitDegrees=45
  twistMotion=Free
 }
 {
  bone1Name=spine_05
  bone2Name=spine_04
  swing1LimitDegrees=45
  swing1Motion=Free
  swing2LimitDegrees=45
  swing2Motion=Free
  twistLimitDegrees=45
  twistMotion=Free
 }
 {
  bone1Name=default_backFlap_c_base
  bone2Name=spine_04
  swing1LimitDegrees=45
  swing1Motion=Free
  swing2LimitDegrees=45
  swing2Motion=Free
  twistLimitDegrees=45
  twistMotion=Free
 }
]
