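# Physics asset for the shirt bottom of the female medium Believer player cosmetic, setting up the bodies and joints for hem secondary motion.
# Eight small base capsules around the front, back and sides attach to a larger spine_02 capsule with free angular motion.
# Each base links to an outer hem capsule with narrow swing and twist limits, slightly tighter at the front, forming short two-body chains.
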PhysicsAsset /BRCosmetics/Characters/Player/Female/Medium/Bodies/F_MED_Believer/Meshes/F_MED_Believer_ShirtsBottom_Physics
bodies {
 default_bottom_bk_a [{shapeType=Capsule, length=3.75, radius=2.5, center=<4 0 0>, rotation=<rot 90 0 0>}]
 default_bottom_bk_a_base [{shapeType=Capsule, length=1, radius=1}]
 default_bottom_bk_a_base_l [{shapeType=Capsule, length=1, radius=1}]
 default_bottom_bk_a_base_r [{shapeType=Capsule, length=1, radius=1}]
 default_bottom_bk_a_l [{shapeType=Capsule, length=3.75, radius=2.5, center=<4 0 0>, rotation=<rot 90 0 0>}]
 default_bottom_bk_a_r [{shapeType=Capsule, length=3.75, radius=2.5, center=<-4 0 0>, rotation=<rot 90 0 0>}]
 default_bottom_fr_a [{shapeType=Capsule, length=3.75, radius=2.5, center=<4 0 0>, rotation=<rot 90 0 0>}]
 default_bottom_fr_a_base [{shapeType=Capsule, length=1, radius=1}]
 default_bottom_fr_a_base_l [{shapeType=Capsule, length=1, radius=1}]
 default_bottom_fr_a_base_r [{shapeType=Capsule, length=1, radius=1}]
 default_bottom_fr_a_l [{shapeType=Capsule, length=3.75, radius=2.5, center=<4 0 0>, rotation=<rot 90 0 0>}]
 default_bottom_fr_a_r [{shapeType=Capsule, length=3.75, radius=2.5, center=<-4 0 0>, rotation=<rot 90 0 0>}]
 default_bottom_side_a_base_l [{shapeType=Capsule, length=1, radius=1}]
 default_bottom_side_a_base_r [{shapeType=Capsule, length=1, radius=1}]
 default_bottom_side_a_l [{shapeType=Capsule, length=3.75, radius=2.5, center=<4 0 0>, rotation=<rot 90 0 0>}]
 default_bottom_side_a_r [{shapeType=Capsule, length=3.75, radius=2.5, center=<-4 0 0>, rotation=<rot 90 0 0>}]
 spine_02 [{shapeType=Capsule, length=6.75, radius=5.75, center=<0.0001 1.3546 0>}]
}
constraints [
 {bone1Name=default_bottom_side_a_l, bone2Name=default_bottom_side_a_base_l, swing1LimitDegrees=25, swing2LimitDegrees=10, twistLimitDegrees=5}
 {bone1Name=default_bottom_fr_a_l, bone2Name=default_bottom_fr_a_base_l, swing1LimitDegrees=20, swing2LimitDegrees=10, twistLimitDegrees=5}
 {bone1Name=default_bottom_bk_a_l, bone2Name=default_bottom_bk_a_base_l, swing1LimitDegrees=25, swing2LimitDegrees=10, twistLimitDegrees=5}
 {bone1Name=default_bottom_side_a_r, bone2Name=default_bottom_side_a_base_r, swing1LimitDegrees=25, swing2LimitDegrees=10, twistLimitDegrees=5}
 {bone1Name=default_bottom_fr_a_r, bone2Name=default_bottom_fr_a_base_r, swing1LimitDegrees=20, swing2LimitDegrees=10, twistLimitDegrees=5}
 {bone1Name=default_bottom_bk_a_r, bone2Name=default_bottom_bk_a_base_r, swing1LimitDegrees=25, swing2LimitDegrees=10, twistLimitDegrees=5}
 {bone1Name=default_bottom_fr_a, bone2Name=default_bottom_fr_a_base, swing1LimitDegrees=20, swing2LimitDegrees=10, twistLimitDegrees=5}
 {bone1Name=default_bottom_bk_a, bone2Name=default_bottom_bk_a_base, swing1LimitDegrees=25, swing2LimitDegrees=10, twistLimitDegrees=5}
 {
  bone1Name=default_bottom_side_a_base_l
  bone2Name=spine_02
  swing1LimitDegrees=45
  swing1Motion=Free
  swing2LimitDegrees=45
  swing2Motion=Free
  twistLimitDegrees=45
  twistMotion=Free
 }
 {
  bone1Name=default_bottom_fr_a_base_l
  bone2Name=spine_02
  swing1LimitDegrees=45
  swing1Motion=Free
  swing2LimitDegrees=45
  swing2Motion=Free
  twistLimitDegrees=45
  twistMotion=Free
 }
 {
  bone1Name=default_bottom_bk_a_base_l
  bone2Name=spine_02
  swing1LimitDegrees=45
  swing1Motion=Free
  swing2LimitDegrees=45
  swing2Motion=Free
  twistLimitDegrees=45
  twistMotion=Free
 }
 {
  bone1Name=default_bottom_side_a_base_r
  bone2Name=spine_02
  swing1LimitDegrees=45
  swing1Motion=Free
  swing2LimitDegrees=45
  swing2Motion=Free
  twistLimitDegrees=45
  twistMotion=Free
 }
 {
  bone1Name=default_bottom_fr_a_base_r
  bone2Name=spine_02
  swing1LimitDegrees=45
  swing1Motion=Free
  swing2LimitDegrees=45
  swing2Motion=Free
  twistLimitDegrees=45
  twistMotion=Free
 }
 {
  bone1Name=default_bottom_bk_a_base_r
  bone2Name=spine_02
  swing1LimitDegrees=45
  swing1Motion=Free
  swing2LimitDegrees=45
  swing2Motion=Free
  twistLimitDegrees=45
  twistMotion=Free
 }
 {
  bone1Name=default_bottom_fr_a_base
  bone2Name=spine_02
  swing1LimitDegrees=45
  swing1Motion=Free
  swing2LimitDegrees=45
  swing2Motion=Free
  twistLimitDegrees=45
  twistMotion=Free
 }
 {
  bone1Name=default_bottom_bk_a_base
  bone2Name=spine_02
  swing1LimitDegrees=45
  swing1Motion=Free
  swing2LimitDegrees=45
  swing2Motion=Free
  twistLimitDegrees=45
  twistMotion=Free
 }
]
