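# Physics asset for the back shirt flaps of the F_MED_Believer_Summer female cosmetic body.
# Three flap capsules hang from small base capsules, with twist locked and narrow, asymmetric swing limits so they flap mostly along one axis.
# The bases and a right tape-strap base attach to spine_04 or spine_05 through angularly free constraints.
# Larger capsules on those two spine bones form the torso volume.
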PhysicsAsset /BRCosmetics/Characters/Player/Female/Medium/Bodies/F_MED_Believer_Summer/Meshes/F_MED_Believer_Summer_ShirtsBack_Physics
bodies {
 deault_backFlap_a [{shapeType=Capsule, length=2, radius=2, center=<3 0 0>, rotation=<rot 90 0 0>}]
 deault_backFlap_a_base [{shapeType=Capsule, length=1, radius=1}]
 default_backFlap_b [{shapeType=Capsule, length=2, radius=2, center=<3 0 0>, rotation=<rot 90 0 0>}]
 default_backFlap_b_base [{shapeType=Capsule, length=1, radius=1}]
 default_backFlap_c [{shapeType=Capsule, length=2, radius=2, center=<3 0 0>, rotation=<rot 90 0 0>}]
 default_backFlap_c_base [{shapeType=Capsule, length=1, radius=1}]
 spine_04 [{shapeType=Capsule, length=2, radius=9.75, center=<0 3.3998 0>}]
 spine_05 [{shapeType=Capsule, length=8.5, radius=7.25, center=<0 1.0972 0>}]
 tapeStrap_a_base_r [{shapeType=Capsule, length=0.505, radius=0.505, rotation=<rot 0 -90 0>}]
}
constraints [
 {
  bone1Name=deault_backFlap_a
  bone2Name=deault_backFlap_a_base
  swing1LimitDegrees=20
  swing2LimitDegrees=5
  twistLimitDegrees=45
  twistMotion=Locked
 }
 {
  bone1Name=default_backFlap_b
  bone2Name=default_backFlap_b_base
  swing1LimitDegrees=20
  swing2LimitDegrees=5
  twistLimitDegrees=45
  twistMotion=Locked
 }
 {
  bone1Name=default_backFlap_c
  bone2Name=default_backFlap_c_base
  swing1LimitDegrees=20
  swing2LimitDegrees=5
  twistLimitDegrees=45
  twistMotion=Locked
 }
 {
  bone1Name=tapeStrap_a_base_r
  bone2Name=spine_05
  swing1LimitDegrees=45
  swing1Motion=Free
  swing2LimitDegrees=45
  swing2Motion=Free
  twistLimitDegrees=45
  twistMotion=Free
 }
 {
  bone1Name=deault_backFlap_a_base
  bone2Name=spine_05
  swing1LimitDegrees=45
  swing1Motion=Free
  swing2LimitDegrees=45
  swing2Motion=Free
  twistLimitDegrees=45
  twistMotion=Free
 }
 {
  bone1Name=default_backFlap_b_base
  bone2Name=spine_05
  swing1LimitDegrees=45
  swing1Motion=Free
  swing2LimitDegrees=45
  swing2Motion=Free
  twistLimitDegrees=45
  twistMotion=Free
 }
 {
  bone1Name=default_backFlap_c_base
  bone2Name=spine_04
  swing1LimitDegrees=45
  swing1Motion=Free
  swing2LimitDegrees=45
  swing2Motion=Free
  twistLimitDegrees=45
  twistMotion=Free
 }
 {
  bone1Name=spine_05
  bone2Name=spine_04
  swing1LimitDegrees=45
  swing1Motion=Free
  swing2LimitDegrees=45
  swing2Motion=Free
  twistLimitDegrees=45
  twistMotion=Free
 }
]
