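# Physics asset for the shirt bottom of the female medium Believer Summer body, setting up secondary motion for the hem.
# Six hem segments (front, back, front left/right, both sides) each pair a capsule body with a small base body, jointed by tight twist and swing limits with a wider primary swing.
# The base bodies attach to spine_03 with free angular motion, and spine_03 links to spine_02, both carrying larger torso capsules.
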PhysicsAsset /BRCosmetics/Characters/Player/Female/Medium/Bodies/F_MED_Believer_Summer/Meshes/F_MED_Believer_Summer_ShirtsBottom_Physics
bodies {
 spine_02 [{shapeType=Capsule, length=5.25, radius=6.25, center=<-0.7504 2.157 0>}]
 spine_03 [{shapeType=Capsule, length=6.5, radius=5.5, center=<-0.0001 3.0304 0>}]
 summer_shirts_bottom_bk [{shapeType=Capsule, length=2.5, radius=2.75, center=<3 0 0>, rotation=<rot 90 0 0>}]
 summer_shirts_bottom_bk_base [{shapeType=Capsule, length=1, radius=1}]
 summer_shirts_bottom_fr [{shapeType=Capsule, length=2.5, radius=2.75, center=<3 0 0>, rotation=<rot 90 0 0>}]
 summer_shirts_bottom_fr_base [{shapeType=Capsule, length=1, radius=1}]
 summer_shirts_bottom_fr_base_l [{shapeType=Capsule, length=1, radius=1}]
 summer_shirts_bottom_fr_base_r [{shapeType=Capsule, length=1, radius=1}]
 summer_shirts_bottom_fr_l [{shapeType=Capsule, length=2.5, radius=2.75, center=<3 0 0>, rotation=<rot 90 0 0>}]
 summer_shirts_bottom_fr_r [{shapeType=Capsule, length=2.5, radius=2.75, center=<-3 0 0>, rotation=<rot 90 0 0>}]
 summer_shirts_side_base_l [{shapeType=Capsule, length=1, radius=1}]
 summer_shirts_side_base_r [{shapeType=Capsule, length=1, radius=1}]
 summer_shirts_side_l [{shapeType=Capsule, length=2.5, radius=2.75, center=<3 0 0>, rotation=<rot 90 0 0>}]
 summer_shirts_side_r [{shapeType=Capsule, length=2.5, radius=2.75, center=<-3 0 0>, rotation=<rot 90 0 0>}]
}
constraints [
 {bone1Name=summer_shirts_bottom_fr, bone2Name=summer_shirts_bottom_fr_base, swing1LimitDegrees=30, swing2LimitDegrees=5, twistLimitDegrees=5}
 {bone1Name=summer_shirts_bottom_bk, bone2Name=summer_shirts_bottom_bk_base, swing1LimitDegrees=35, swing2LimitDegrees=5, twistLimitDegrees=5}
 {
  bone1Name=summer_shirts_bottom_fr_l
  bone2Name=summer_shirts_bottom_fr_base_l
  swing1LimitDegrees=30
  swing2LimitDegrees=5
  twistLimitDegrees=5
 }
 {bone1Name=summer_shirts_side_l, bone2Name=summer_shirts_side_base_l, swing1LimitDegrees=30, swing2LimitDegrees=5, twistLimitDegrees=5}
 {
  bone1Name=summer_shirts_bottom_fr_r
  bone2Name=summer_shirts_bottom_fr_base_r
  swing1LimitDegrees=30
  swing2LimitDegrees=5
  twistLimitDegrees=5
 }
 {bone1Name=summer_shirts_side_r, bone2Name=summer_shirts_side_base_r, swing1LimitDegrees=30, swing2LimitDegrees=5, twistLimitDegrees=5}
 {
  bone1Name=summer_shirts_bottom_fr_base
  bone2Name=spine_03
  swing1LimitDegrees=45
  swing1Motion=Free
  swing2LimitDegrees=45
  swing2Motion=Free
  twistLimitDegrees=45
  twistMotion=Free
 }
 {
  bone1Name=summer_shirts_bottom_bk_base
  bone2Name=spine_03
  swing1LimitDegrees=45
  swing1Motion=Free
  swing2LimitDegrees=45
  swing2Motion=Free
  twistLimitDegrees=45
  twistMotion=Free
 }
 {
  bone1Name=summer_shirts_bottom_fr_base_l
  bone2Name=spine_03
  swing1LimitDegrees=45
  swing1Motion=Free
  swing2LimitDegrees=45
  swing2Motion=Free
  twistLimitDegrees=45
  twistMotion=Free
 }
 {
  bone1Name=summer_shirts_side_base_l
  bone2Name=spine_03
  swing1LimitDegrees=45
  swing1Motion=Free
  swing2LimitDegrees=45
  swing2Motion=Free
  twistLimitDegrees=45
  twistMotion=Free
 }
 {
  bone1Name=summer_shirts_bottom_fr_base_r
  bone2Name=spine_03
  swing1LimitDegrees=45
  swing1Motion=Free
  swing2LimitDegrees=45
  swing2Motion=Free
  twistLimitDegrees=45
  twistMotion=Free
 }
 {
  bone1Name=summer_shirts_side_base_r
  bone2Name=spine_03
  swing1LimitDegrees=45
  swing1Motion=Free
  swing2LimitDegrees=45
  swing2Motion=Free
  twistLimitDegrees=45
  twistMotion=Free
 }
 {
  bone1Name=spine_03
  bone2Name=spine_02
  swing1LimitDegrees=45
  swing1Motion=Free
  swing2LimitDegrees=45
  swing2Motion=Free
  twistLimitDegrees=45
  twistMotion=Free
 }
]
