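# Physics asset for the strap mesh of the female medium Believer Summer cosmetic body, defining capsule bodies for tape straps and band flaps.
# Each side has a three-link tape strap chain and two single-link band flaps, attached through tiny base capsules to a large spine_05 torso capsule.
# Base-to-spine joints rotate freely, while strap links have narrow limits and flaps very tight ones.
# Left tape strap joints also lock twist, unlike the right.
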PhysicsAsset /BRCosmetics/Characters/Player/Female/Medium/Bodies/F_MED_Believer_Summer/Meshes/F_MED_Believer_Summer_Strap_Physics
bodies {
 bandFlap_a_base_l [{shapeType=Capsule, length=0.505, radius=0.505, rotation=<rot 0 -90 0>}]
 bandFlap_a_base_r [{shapeType=Capsule, length=0.505, radius=0.505, rotation=<rot 0 -90 0>}]
 bandFlap_a_l [{shapeType=Capsule, length=1.75, radius=1.25, center=<1 0 0>, rotation=<rot 90 0 0>}]
 bandFlap_a_r [{shapeType=Capsule, length=2.25, radius=1, center=<1 0 0>, rotation=<rot 90 0 0>}]
 bandFlap_b_base_l [{shapeType=Capsule, length=0.505, radius=0.505, rotation=<rot 0 -90 0>}]
 bandFlap_b_base_r [{shapeType=Capsule, length=0.505, radius=0.505, rotation=<rot 0 -90 0>}]
 bandFlap_b_l [{shapeType=Capsule, length=1.75, radius=1.25, center=<1 0 0>, rotation=<rot 90 0 0>}]
 bandFlap_b_r [{shapeType=Capsule, length=2.25, radius=1, center=<1 0 0>, rotation=<rot 90 0 0>}]
 spine_05 [{shapeType=Capsule, length=9.75, radius=8.5, center=<2.3369 0.218 0.0001>}]
 tapeStrap_a_base_l [{shapeType=Capsule, length=0.505, radius=0.505, rotation=<rot 0 -90 0>}]
 tapeStrap_a_base_r [{shapeType=Capsule, length=0.505, radius=0.505, rotation=<rot 0 -90 0>}]
 tapeStrap_a_l [{shapeType=Capsule, length=4.5, radius=1.25, center=<3 0 0>, rotation=<rot 90 0 0>}]
 tapeStrap_a_r [{shapeType=Capsule, length=4.5, radius=1.25, center=<3 0 0>, rotation=<rot 90 0 0>}]
 tapeStrap_b_l [{shapeType=Capsule, length=3, radius=1, center=<2.3 0 0>, rotation=<rot 90 0 0>}]
 tapeStrap_b_r [{shapeType=Capsule, length=3, radius=1, center=<2.3 0 0>, rotation=<rot 90 0 0>}]
 tapeStrap_c_l [{shapeType=Capsule, length=2.25, radius=1.25, center=<2.3 0 0>, rotation=<rot 90 0 0>}]
 tapeStrap_c_r [{shapeType=Capsule, length=2.25, radius=1.25, center=<2.3 0 0>, rotation=<rot 90 0 0>}]
}
constraints [
 {
  bone1Name=tapeStrap_a_base_r
  bone2Name=spine_05
  swing1LimitDegrees=45
  swing1Motion=Free
  swing2LimitDegrees=45
  swing2Motion=Free
  twistLimitDegrees=45
  twistMotion=Free
 }
 {
  bone1Name=tapeStrap_a_base_l
  bone2Name=spine_05
  swing1LimitDegrees=45
  swing1Motion=Free
  swing2LimitDegrees=45
  swing2Motion=Free
  twistLimitDegrees=45
  twistMotion=Free
 }
 {bone1Name=tapeStrap_a_r, bone2Name=tapeStrap_a_base_r, swing1LimitDegrees=20, swing2LimitDegrees=5, twistLimitDegrees=5}
 {bone1Name=tapeStrap_b_r, bone2Name=tapeStrap_a_r, swing1LimitDegrees=25, swing2LimitDegrees=10, twistLimitDegrees=5}
 {bone1Name=tapeStrap_c_r, bone2Name=tapeStrap_b_r, swing1LimitDegrees=25, swing2LimitDegrees=10, twistLimitDegrees=5}
 {bone1Name=tapeStrap_a_l, bone2Name=tapeStrap_a_base_l, swing1LimitDegrees=20, swing2LimitDegrees=5, twistLimitDegrees=5, twistMotion=Locked}
 {bone1Name=tapeStrap_b_l, bone2Name=tapeStrap_a_l, swing1LimitDegrees=15, swing2LimitDegrees=15, twistLimitDegrees=5, twistMotion=Locked}
 {bone1Name=tapeStrap_c_l, bone2Name=tapeStrap_b_l, swing1LimitDegrees=15, swing2LimitDegrees=15, twistLimitDegrees=5, twistMotion=Locked}
 {
  bone1Name=bandFlap_a_base_l
  bone2Name=spine_05
  swing1LimitDegrees=45
  swing1Motion=Free
  swing2LimitDegrees=45
  swing2Motion=Free
  twistLimitDegrees=45
  twistMotion=Free
 }
 {
  bone1Name=bandFlap_b_base_l
  bone2Name=spine_05
  swing1LimitDegrees=45
  swing1Motion=Free
  swing2LimitDegrees=45
  swing2Motion=Free
  twistLimitDegrees=45
  twistMotion=Free
 }
 {bone1Name=bandFlap_a_l, bone2Name=bandFlap_a_base_l, swing1LimitDegrees=3, swing2LimitDegrees=3, twistLimitDegrees=3}
 {bone1Name=bandFlap_b_l, bone2Name=bandFlap_b_base_l, swing1LimitDegrees=3, swing2LimitDegrees=3, twistLimitDegrees=3}
 {
  bone1Name=bandFlap_a_base_r
  bone2Name=spine_05
  swing1LimitDegrees=45
  swing1Motion=Free
  swing2LimitDegrees=45
  swing2Motion=Free
  twistLimitDegrees=45
  twistMotion=Free
 }
 {
  bone1Name=bandFlap_b_base_r
  bone2Name=spine_05
  swing1LimitDegrees=45
  swing1Motion=Free
  swing2LimitDegrees=45
  swing2Motion=Free
  twistLimitDegrees=45
  twistMotion=Free
 }
 {bone1Name=bandFlap_a_r, bone2Name=bandFlap_a_base_r, swing1LimitDegrees=3, swing2LimitDegrees=3, twistLimitDegrees=3}
 {bone1Name=bandFlap_b_r, bone2Name=bandFlap_b_base_r, swing1LimitDegrees=3, swing2LimitDegrees=3, twistLimitDegrees=3}
]
