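# Physics asset for the female medium Believer Summer player body, setting up capsule bodies for secondary motion of shirt hem, back flaps and tape straps.
# Small base bodies join spine_03 or spine_05 with free angular constraints, and each drives a flap or strap through tight swing limits.
# Tape straps form three-link chains per side, with the left tighter and twist-locked.
# Back flaps also lock twist, while the six shirt pieces allow slight twist.
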PhysicsAsset /Game/Characters/Player/Female/Medium/Bodies/F_MED_Believer_Summer/Meshes/F_MED_Believer_Summer_Physics
bodies {
 deault_backFlap_a [{shapeType=Capsule, length=2, radius=2, center=<3 0 0>, rotation=<rot 90 0 0>}]
 deault_backFlap_a_base [{shapeType=Capsule, length=1, radius=1}]
 default_backFlap_b [{shapeType=Capsule, length=2, radius=2, center=<3 0 0>, rotation=<rot 90 0 0>}]
 default_backFlap_b_base [{shapeType=Capsule, length=1, radius=1}]
 default_backFlap_c [{shapeType=Capsule, length=2, radius=2, center=<3 0 0>, rotation=<rot 90 0 0>}]
 default_backFlap_c_base [{shapeType=Capsule, length=1, radius=1}]
 spine_02 [{shapeType=Capsule, length=5.25, radius=6.25, center=<-0.7504 2.157 0>}]
 spine_03 [{shapeType=Capsule, length=6.5, radius=5.5, center=<-0.0001 3.0304 0>}]
 spine_05 [{shapeType=Capsule, length=9.75, radius=8.5, center=<2.3369 0.218 0.0001>}]
 summer_shirts_bottom_bk [{shapeType=Capsule, length=2.5, radius=2.75, center=<3 0 0>, rotation=<rot 90 0 0>}]
 summer_shirts_bottom_bk_base [{shapeType=Capsule, length=1, radius=1}]
 summer_shirts_bottom_fr [{shapeType=Capsule, length=2.5, radius=2.75, center=<3 0 0>, rotation=<rot 90 0 0>}]
 summer_shirts_bottom_fr_base [{shapeType=Capsule, length=1, radius=1}]
 summer_shirts_bottom_fr_base_l [{shapeType=Capsule, length=1, radius=1}]
 summer_shirts_bottom_fr_base_r [{shapeType=Capsule, length=1, radius=1}]
 summer_shirts_bottom_fr_l [{shapeType=Capsule, length=2.5, radius=2.75, center=<3 0 0>, rotation=<rot 90 0 0>}]
 summer_shirts_bottom_fr_r [{shapeType=Capsule, length=2.5, radius=2.75, center=<-3 0 0>, rotation=<rot 90 0 0>}]
 summer_shirts_side_base_l [{shapeType=Capsule, length=1, radius=1}]
 summer_shirts_side_base_r [{shapeType=Capsule, length=1, radius=1}]
 summer_shirts_side_l [{shapeType=Capsule, length=2.5, radius=2.75, center=<3 0 0>, rotation=<rot 90 0 0>}]
 summer_shirts_side_r [{shapeType=Capsule, length=2.5, radius=2.75, center=<-3 0 0>, rotation=<rot 90 0 0>}]
 tapeStrap_a_base_l [{shapeType=Capsule, length=0.505, radius=0.505, rotation=<rot 0 -90 0>}]
 tapeStrap_a_base_r [{shapeType=Capsule, length=0.505, radius=0.505, rotation=<rot 0 -90 0>}]
 tapeStrap_a_l [{shapeType=Capsule, length=3, radius=1.25, center=<1.7 0 0>, rotation=<rot 90 0 0>}]
 tapeStrap_a_r [{shapeType=Capsule, length=3, radius=1.25, center=<1.7 0 0>, rotation=<rot 90 0 0>}]
 tapeStrap_b_l [{shapeType=Capsule, length=2.75, radius=1, center=<2 0 0>, rotation=<rot 90 0 0>}]
 tapeStrap_b_r [{shapeType=Capsule, length=2.75, radius=1, center=<2 0 0>, rotation=<rot 90 0 0>}]
 tapeStrap_c_l [{shapeType=Capsule, length=2.25, radius=1.25, center=<2.3 0 0>, rotation=<rot 90 0 0>}]
 tapeStrap_c_r [{shapeType=Capsule, length=2.25, radius=1.25, center=<2.3 0 0>, rotation=<rot 90 0 0>}]
}
constraints [
 {bone1Name=summer_shirts_bottom_fr, bone2Name=summer_shirts_bottom_fr_base, swing1LimitDegrees=30, swing2LimitDegrees=5, twistLimitDegrees=5}
 {bone1Name=summer_shirts_bottom_bk, bone2Name=summer_shirts_bottom_bk_base, swing1LimitDegrees=35, swing2LimitDegrees=5, twistLimitDegrees=5}
 {
  bone1Name=summer_shirts_bottom_fr_l
  bone2Name=summer_shirts_bottom_fr_base_l
  swing1LimitDegrees=30
  swing2LimitDegrees=5
  twistLimitDegrees=5
 }
 {bone1Name=summer_shirts_side_l, bone2Name=summer_shirts_side_base_l, swing1LimitDegrees=30, swing2LimitDegrees=5, twistLimitDegrees=5}
 {
  bone1Name=summer_shirts_bottom_fr_r
  bone2Name=summer_shirts_bottom_fr_base_r
  swing1LimitDegrees=30
  swing2LimitDegrees=5
  twistLimitDegrees=5
 }
 {bone1Name=summer_shirts_side_r, bone2Name=summer_shirts_side_base_r, swing1LimitDegrees=30, swing2LimitDegrees=5, twistLimitDegrees=5}
 {
  bone1Name=summer_shirts_bottom_fr_base
  bone2Name=spine_03
  swing1LimitDegrees=45
  swing1Motion=Free
  swing2LimitDegrees=45
  swing2Motion=Free
  twistLimitDegrees=45
  twistMotion=Free
 }
 {
  bone1Name=summer_shirts_bottom_bk_base
  bone2Name=spine_03
  swing1LimitDegrees=45
  swing1Motion=Free
  swing2LimitDegrees=45
  swing2Motion=Free
  twistLimitDegrees=45
  twistMotion=Free
 }
 {
  bone1Name=summer_shirts_bottom_fr_base_l
  bone2Name=spine_03
  swing1LimitDegrees=45
  swing1Motion=Free
  swing2LimitDegrees=45
  swing2Motion=Free
  twistLimitDegrees=45
  twistMotion=Free
 }
 {
  bone1Name=summer_shirts_side_base_l
  bone2Name=spine_03
  swing1LimitDegrees=45
  swing1Motion=Free
  swing2LimitDegrees=45
  swing2Motion=Free
  twistLimitDegrees=45
  twistMotion=Free
 }
 {
  bone1Name=summer_shirts_bottom_fr_base_r
  bone2Name=spine_03
  swing1LimitDegrees=45
  swing1Motion=Free
  swing2LimitDegrees=45
  swing2Motion=Free
  twistLimitDegrees=45
  twistMotion=Free
 }
 {
  bone1Name=summer_shirts_side_base_r
  bone2Name=spine_03
  swing1LimitDegrees=45
  swing1Motion=Free
  swing2LimitDegrees=45
  swing2Motion=Free
  twistLimitDegrees=45
  twistMotion=Free
 }
 {
  bone1Name=spine_03
  bone2Name=spine_02
  swing1LimitDegrees=45
  swing1Motion=Free
  swing2LimitDegrees=45
  swing2Motion=Free
  twistLimitDegrees=45
  twistMotion=Free
 }
 {
  bone1Name=default_backFlap_c_base
  bone2Name=spine_03
  swing1LimitDegrees=45
  swing1Motion=Free
  swing2LimitDegrees=45
  swing2Motion=Free
  twistLimitDegrees=45
  twistMotion=Free
 }
 {
  bone1Name=deault_backFlap_a_base
  bone2Name=spine_05
  swing1LimitDegrees=45
  swing1Motion=Free
  swing2LimitDegrees=45
  swing2Motion=Free
  twistLimitDegrees=45
  twistMotion=Free
 }
 {
  bone1Name=default_backFlap_b_base
  bone2Name=spine_05
  swing1LimitDegrees=45
  swing1Motion=Free
  swing2LimitDegrees=45
  swing2Motion=Free
  twistLimitDegrees=45
  twistMotion=Free
 }
 {
  bone1Name=deault_backFlap_a
  bone2Name=deault_backFlap_a_base
  swing1LimitDegrees=20
  swing2LimitDegrees=5
  twistLimitDegrees=45
  twistMotion=Locked
 }
 {
  bone1Name=default_backFlap_b
  bone2Name=default_backFlap_b_base
  swing1LimitDegrees=20
  swing2LimitDegrees=5
  twistLimitDegrees=45
  twistMotion=Locked
 }
 {
  bone1Name=default_backFlap_c
  bone2Name=default_backFlap_c_base
  swing1LimitDegrees=20
  swing2LimitDegrees=5
  twistLimitDegrees=45
  twistMotion=Locked
 }
 {
  bone1Name=tapeStrap_a_base_r
  bone2Name=spine_05
  swing1LimitDegrees=45
  swing1Motion=Free
  swing2LimitDegrees=45
  swing2Motion=Free
  twistLimitDegrees=45
  twistMotion=Free
 }
 {
  bone1Name=tapeStrap_a_base_l
  bone2Name=spine_05
  swing1LimitDegrees=45
  swing1Motion=Free
  swing2LimitDegrees=45
  swing2Motion=Free
  twistLimitDegrees=45
  twistMotion=Free
 }
 {bone1Name=tapeStrap_a_r, bone2Name=tapeStrap_a_base_r, swing1LimitDegrees=25, swing2LimitDegrees=5, twistLimitDegrees=5}
 {bone1Name=tapeStrap_b_r, bone2Name=tapeStrap_a_r, swing1LimitDegrees=25, swing2LimitDegrees=5, twistLimitDegrees=5}
 {bone1Name=tapeStrap_c_r, bone2Name=tapeStrap_b_r, swing1LimitDegrees=25, swing2LimitDegrees=5, twistLimitDegrees=5}
 {bone1Name=tapeStrap_a_l, bone2Name=tapeStrap_a_base_l, swing1LimitDegrees=15, swing2LimitDegrees=2, twistLimitDegrees=5, twistMotion=Locked}
 {bone1Name=tapeStrap_b_l, bone2Name=tapeStrap_a_l, swing1LimitDegrees=15, swing2LimitDegrees=2, twistLimitDegrees=5, twistMotion=Locked}
 {bone1Name=tapeStrap_c_l, bone2Name=tapeStrap_b_l, swing1LimitDegrees=15, swing2LimitDegrees=2, twistLimitDegrees=5, twistMotion=Locked}
 {
  bone1Name=spine_05
  bone2Name=spine_03
  swing1LimitDegrees=45
  swing1Motion=Free
  swing2LimitDegrees=45
  swing2Motion=Free
  twistLimitDegrees=45
  twistMotion=Free
 }
]
