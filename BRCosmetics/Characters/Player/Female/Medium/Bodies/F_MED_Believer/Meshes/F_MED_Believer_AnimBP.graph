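# Post-copy Animation Blueprint for the F_MED_Believer cosmetic body, adding muscle volume and secondary motion for its outfit variants.
# Copies the attached parent's pose through a muscle-volume Control Rig, then caches it for the variant branches.
# Branches use weighted bone constraints, AnimDynamics and corrective bone transforms on parts like flaps, zippers and glutes.
# The variantIndex variable blends between the default, summer and jacket cached poses for output.
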AnimBlueprint /BRCosmetics/Characters/Player/Female/Medium/Bodies/F_MED_Believer/Meshes/F_MED_Believer_AnimBP
target_skeleton=&/F_MED_Believer_Skeleton
graphs [
 AnimGraph {nodes [
   n1 = Anim Dynamics - Bone: default_bottom_fr_a(Alpha=1, ExternalForce="0,0,0") {
    angularDampingOverride=1
    angularSpringConstant=1500
    bAngularSpring=true
    bGravityOverrideInSimSpace=true
    bOverrideAngularDamping=true
    gravityScale=0.35
    simulationSpace=RootRelative
    alphaBoolBlend {blendOption=Linear}
    boundBone {boneName=default_bottom_fr_a}
    componentLinearAccScale=<0.2 0.2 0.2>
    componentLinearVelScale=<0.2 0.2 0.2>
    physicsBodyDefinitions [{
      sphereCollisionRadius=0
      boundBone {boneName=default_bottom_fr_a}
      boxExtents=<10 20 20>
      constraintSetup {angularLimitsMax=<0 0 0>, angularLimitsMin=<0 0 -5>, angularTarget=<0 0 0>}
      localJointOffset=<5 0 0>
     }]
   }
   n2 = Anim Dynamics - Bone: default_bottom_fr_b(ComponentPose=@n1, Alpha=1, ExternalForce="0,0,0") {
    angularDampingOverride=1
    angularSpringConstant=1500
    bAngularSpring=true
    bGravityOverrideInSimSpace=true
    bOverrideAngularDamping=true
    bUsePlanarLimit=false
    gravityScale=0.35
    simulationSpace=RootRelative
    alphaBoolBlend {blendOption=Linear}
    boundBone {boneName=default_bottom_fr_b}
    componentLinearAccScale=<0.2 0.2 0.2>
    componentLinearVelScale=<0.2 0.2 0.2>
    physicsBodyDefinitions [{
      sphereCollisionRadius=0
      boundBone {boneName=default_bottom_fr_b}
      boxExtents=<10 20 20>
      constraintSetup {angularLimitsMax=<0 5 10>, angularLimitsMin=<0 -5 -15>, angularTarget=<0 0 0>}
      localJointOffset=<5 0 0>
     }]
   }
   n3 = Anim Dynamics - Bone: default_bottom_bk_a(ComponentPose=@n2, Alpha=1, ExternalForce="0,0,0") {
    angularDampingOverride=1
    angularSpringConstant=1500
    bAngularSpring=true
    bGravityOverrideInSimSpace=true
    bOverrideAngularDamping=true
    gravityScale=0.35
    simulationSpace=RootRelative
    alphaBoolBlend {blendOption=Linear}
    boundBone {boneName=default_bottom_bk_a}
    componentLinearAccScale=<0.2 0.2 0.5>
    componentLinearVelScale=<0.2 0.2 0.5>
    physicsBodyDefinitions [{
      sphereCollisionRadius=0
      boundBone {boneName=default_bottom_bk_a}
      boxExtents=<10 20 20>
      constraintSetup {angularLimitsMax=<0 0 0>, angularLimitsMin=<0 0 -10>, angularTarget=<0 0 0>}
      localJointOffset=<5 0 0>
     }]
   }
   n4 = Anim Dynamics - Bone: default_bottom_bk_b(ComponentPose=@n3, Alpha=1, ExternalForce="0,0,0") {
    angularDampingOverride=1
    angularSpringConstant=1500
    bAngularSpring=true
    bGravityOverrideInSimSpace=true
    bOverrideAngularDamping=true
    gravityScale=0.35
    simulationSpace=RootRelative
    alphaBoolBlend {blendOption=Linear}
    boundBone {boneName=default_bottom_bk_b}
    componentLinearAccScale=<0.2 0.2 0.5>
    componentLinearVelScale=<0.2 0.2 0.5>
    physicsBodyDefinitions [{
      sphereCollisionRadius=0
      boundBone {boneName=default_bottom_bk_b}
      boxExtents=<10 20 20>
      constraintSetup {angularLimitsMax=<0 5 10>, angularLimitsMin=<0 -5 -15>, angularTarget=<0 0 0>}
      localJointOffset=<5 0 0>
     }]
   }
   n5 = Anim Dynamics - Bone: default_bottom_fr_a_l(ComponentPose=@n4, Alpha=1, ExternalForce="0,0,0") {
    angularDampingOverride=1
    angularSpringConstant=1500
    bAngularSpring=true
    bGravityOverrideInSimSpace=true
    bOverrideAngularDamping=true
    gravityScale=0.35
    simulationSpace=RootRelative
    alphaBoolBlend {blendOption=Linear}
    boundBone {boneName=default_bottom_fr_a_l}
    componentLinearAccScale=<0.2 0.2 0.2>
    componentLinearVelScale=<0.2 0.2 0.2>
    physicsBodyDefinitions [{
      sphereCollisionRadius=0
      boundBone {boneName=default_bottom_fr_a_l}
      boxExtents=<10 20 20>
      constraintSetup {angularLimitsMax=<0 0 0>, angularLimitsMin=<0 0 -5>, angularTarget=<0 0 0>}
      localJointOffset=<5 0 0>
     }]
   }
   n6 = Anim Dynamics - Bone: default_bottom_fr_b_l(ComponentPose=@n5, Alpha=1, ExternalForce="0,0,0") {
    angularDampingOverride=1
    angularSpringConstant=1500
    bAngularSpring=true
    bGravityOverrideInSimSpace=true
    bOverrideAngularDamping=true
    gravityScale=0.35
    simulationSpace=RootRelative
    alphaBoolBlend {blendOption=Linear}
    boundBone {boneName=default_bottom_fr_b_l}
    componentLinearAccScale=<0.2 0.2 0.2>
    componentLinearVelScale=<0.2 0.2 0.2>
    physicsBodyDefinitions [{
      sphereCollisionRadius=0
      boundBone {boneName=default_bottom_fr_b_l}
      boxExtents=<10 20 20>
      constraintSetup {angularLimitsMax=<0 0 0>, angularLimitsMin=<0 0 -5>, angularTarget=<0 0 0>}
      localJointOffset=<5 0 0>
     }]
   }
   n7 = Anim Dynamics - Bone: default_bottom_side_a_l(ComponentPose=@n6, Alpha=1, ExternalForce="0,0,0") {
    angularDampingOverride=1
    angularSpringConstant=1500
    bAngularSpring=true
    bGravityOverrideInSimSpace=true
    bOverrideAngularDamping=true
    gravityScale=0.35
    simulationSpace=RootRelative
    alphaBoolBlend {blendOption=Linear}
    boundBone {boneName=default_bottom_side_a_l}
    componentLinearAccScale=<0.2 0.2 0.5>
    componentLinearVelScale=<0.2 0.2 0.5>
    physicsBodyDefinitions [{
      sphereCollisionRadius=0
      boundBone {boneName=default_bottom_side_a_l}
      boxExtents=<10 20 20>
      constraintSetup {angularLimitsMax=<0 0 0>, angularLimitsMin=<0 0 -5>, angularTarget=<0 0 0>}
      localJointOffset=<5 0 0>
     }]
   }
   n8 = Anim Dynamics - Bone: default_bottom_side_b_l(ComponentPose=@n7, Alpha=1, ExternalForce="0,0,0") {
    angularDampingOverride=1
    angularSpringConstant=1500
    bAngularSpring=true
    bGravityOverrideInSimSpace=true
    bOverrideAngularDamping=true
    gravityScale=0.35
    simulationSpace=RootRelative
    alphaBoolBlend {blendOption=Linear}
    boundBone {boneName=default_bottom_side_b_l}
    componentLinearAccScale=<0.2 0.2 0.5>
    componentLinearVelScale=<0.2 0.2 0.5>
    physicsBodyDefinitions [{
      sphereCollisionRadius=0
      boundBone {boneName=default_bottom_side_b_l}
      boxExtents=<10 20 20>
      constraintSetup {angularLimitsMax=<0 0 0>, angularLimitsMin=<0 0 -5>, angularTarget=<0 0 0>}
      localJointOffset=<5 0 0>
     }]
   }
   n9 = Anim Dynamics - Bone: default_bottom_bk_a_l(ComponentPose=@n8, Alpha=1, ExternalForce="0,0,0") {
    angularDampingOverride=1
    angularSpringConstant=1500
    bAngularSpring=true
    bGravityOverrideInSimSpace=true
    bOverrideAngularDamping=true
    gravityScale=0.35
    simulationSpace=RootRelative
    alphaBoolBlend {blendOption=Linear}
    boundBone {boneName=default_bottom_bk_a_l}
    componentLinearAccScale=<0.2 0.2 0.5>
    componentLinearVelScale=<0.2 0.2 0.5>
    physicsBodyDefinitions [{
      sphereCollisionRadius=0
      boundBone {boneName=default_bottom_bk_a_l}
      boxExtents=<10 20 20>
      constraintSetup {angularLimitsMax=<0 0 0>, angularLimitsMin=<0 0 -5>, angularTarget=<0 0 0>}
      localJointOffset=<5 0 0>
     }]
   }
   n10 = Anim Dynamics - Bone: default_bottom_bk_b_l(ComponentPose=@n9, Alpha=1, ExternalForce="0,0,0") {
    angularDampingOverride=1
    angularSpringConstant=1500
    bAngularSpring=true
    bGravityOverrideInSimSpace=true
    bOverrideAngularDamping=true
    gravityScale=0.35
    simulationSpace=RootRelative
    alphaBoolBlend {blendOption=Linear}
    boundBone {boneName=default_bottom_bk_b_l}
    componentLinearAccScale=<0.2 0.2 0.5>
    componentLinearVelScale=<0.2 0.2 0.5>
    physicsBodyDefinitions [{
      sphereCollisionRadius=0
      boundBone {boneName=default_bottom_bk_b_l}
      boxExtents=<10 20 20>
      constraintSetup {angularLimitsMax=<0 0 0>, angularLimitsMin=<0 0 -5>, angularTarget=<0 0 0>}
      localJointOffset=<5 0 0>
     }]
   }
   n11 = Anim Dynamics - Bone: default_bottom_fr_a_r(ComponentPose=@n10, Alpha=1, ExternalForce="0,0,0") {
    angularDampingOverride=1
    angularSpringConstant=1500
    bAngularSpring=true
    bGravityOverrideInSimSpace=true
    bOverrideAngularDamping=true
    gravityScale=0.35
    simulationSpace=RootRelative
    alphaBoolBlend {blendOption=Linear}
    boundBone {boneName=default_bottom_fr_a_r}
    componentLinearAccScale=<0.2 0.2 0.2>
    componentLinearVelScale=<0.2 0.2 0.2>
    physicsBodyDefinitions [{
      sphereCollisionRadius=0
      boundBone {boneName=default_bottom_fr_a_r}
      boxExtents=<10 20 20>
      constraintSetup {angularLimitsMax=<0 0 0>, angularLimitsMin=<0 0 -5>, angularTarget=<0 0 0>}
      localJointOffset=<-5 0 0>
     }]
   }
   n12 = Anim Dynamics - Bone: default_bottom_fr_b_r(ComponentPose=@n11, Alpha=1, ExternalForce="0,0,0") {
    angularDampingOverride=1
    angularSpringConstant=1500
    bAngularSpring=true
    bGravityOverrideInSimSpace=true
    bOverrideAngularDamping=true
    gravityScale=0.35
    simulationSpace=RootRelative
    alphaBoolBlend {blendOption=Linear}
    boundBone {boneName=default_bottom_fr_b_r}
    componentLinearAccScale=<0.2 0.2 0.2>
    componentLinearVelScale=<0.2 0.2 0.2>
    physicsBodyDefinitions [{
      sphereCollisionRadius=0
      boundBone {boneName=default_bottom_fr_b_r}
      boxExtents=<10 20 20>
      constraintSetup {angularLimitsMax=<0 0 0>, angularLimitsMin=<0 0 -5>, angularTarget=<0 0 0>}
      localJointOffset=<-5 0 0>
     }]
   }
   n13 = Anim Dynamics - Bone: default_bottom_side_a_r(ComponentPose=@n12, Alpha=1, ExternalForce="0,0,0") {
    angularDampingOverride=1
    angularSpringConstant=1500
    bAngularSpring=true
    bGravityOverrideInSimSpace=true
    bOverrideAngularDamping=true
    gravityScale=0.35
    simulationSpace=RootRelative
    alphaBoolBlend {blendOption=Linear}
    boundBone {boneName=default_bottom_side_a_r}
    componentLinearAccScale=<0.2 0.2 0.5>
    componentLinearVelScale=<0.2 0.2 0.5>
    physicsBodyDefinitions [{
      sphereCollisionRadius=0
      boundBone {boneName=default_bottom_side_a_r}
      boxExtents=<10 20 20>
      constraintSetup {angularLimitsMax=<0 0 0>, angularLimitsMin=<0 0 -5>, angularTarget=<0 0 0>}
      localJointOffset=<-5 0 0>
     }]
   }
   n14 = Anim Dynamics - Bone: default_bottom_side_b_r(ComponentPose=@n13, Alpha=1, ExternalForce="0,0,0") {
    angularDampingOverride=1
    angularSpringConstant=1500
    bAngularSpring=true
    bGravityOverrideInSimSpace=true
    bOverrideAngularDamping=true
    gravityScale=0.35
    simulationSpace=RootRelative
    alphaBoolBlend {blendOption=Linear}
    boundBone {boneName=default_bottom_side_b_r}
    componentLinearAccScale=<0.2 0.2 0.5>
    componentLinearVelScale=<0.2 0.2 0.5>
    physicsBodyDefinitions [{
      sphereCollisionRadius=0
      boundBone {boneName=default_bottom_side_b_r}
      boxExtents=<10 20 20>
      constraintSetup {angularLimitsMax=<0 0 0>, angularLimitsMin=<0 0 -5>, angularTarget=<0 0 0>}
      localJointOffset=<-5 0 0>
     }]
   }
   n15 = Anim Dynamics - Bone: default_bottom_bk_a_r(ComponentPose=@n14, Alpha=1, ExternalForce="0,0,0") {
    angularDampingOverride=1
    angularSpringConstant=1500
    bAngularSpring=true
    bGravityOverrideInSimSpace=true
    bOverrideAngularDamping=true
    gravityScale=0.35
    simulationSpace=RootRelative
    alphaBoolBlend {blendOption=Linear}
    boundBone {boneName=default_bottom_bk_a_r}
    componentLinearAccScale=<0.2 0.2 0.5>
    componentLinearVelScale=<0.2 0.2 0.5>
    physicsBodyDefinitions [{
      sphereCollisionRadius=0
      boundBone {boneName=default_bottom_bk_a_r}
      boxExtents=<10 20 20>
      constraintSetup {angularLimitsMax=<0 0 0>, angularLimitsMin=<0 0 -5>, angularTarget=<0 0 0>}
      localJointOffset=<-5 0 0>
     }]
   }
   Anim Dynamics - Bone: default_bottom_bk_b_r(ComponentPose=@n15, Alpha=1, ExternalForce="0,0,0") {
    angularDampingOverride=1
    angularSpringConstant=1500
    bAngularSpring=true
    bGravityOverrideInSimSpace=true
    bOverrideAngularDamping=true
    gravityScale=0.35
    simulationSpace=RootRelative
    alphaBoolBlend {blendOption=Linear}
    boundBone {boneName=default_bottom_bk_b_r}
    componentLinearAccScale=<0.2 0.2 0.5>
    componentLinearVelScale=<0.2 0.2 0.5>
    physicsBodyDefinitions [{
      sphereCollisionRadius=0
      boundBone {boneName=default_bottom_bk_b_r}
      boxExtents=<10 20 20>
      constraintSetup {angularLimitsMax=<0 0 0>, angularLimitsMin=<0 0 -5>, angularTarget=<0 0 0>}
      localJointOffset=<-5 0 0>
     }]
   }
   n16 = Anim Dynamics - Bone: bandFlap_a_l(Alpha=0.25, ExternalForce="0,0,0") {
    angularDampingOverride=1
    angularSpringConstant=1500
    bAngularSpring=true
    bGravityOverrideInSimSpace=true
    bOverrideAngularDamping=true
    gravityScale=0.35
    simulationSpace=RootRelative
    alphaBoolBlend {blendOption=Linear}
    boundBone {boneName=bandFlap_a_l}
    physicsBodyDefinitions [{
      sphereCollisionRadius=0
      boundBone {boneName=bandFlap_a_l}
      boxExtents=<10 20 20>
      constraintSetup {angularLimitsMax=<0 20 30>, angularLimitsMin=<0 -20 -30>, angularTarget=<0 0 0>}
      localJointOffset=<5 0 0>
     }]
   }
   Anim Dynamics - Bone: bandFlap_b_l(ComponentPose=@n16, Alpha=0.25, ExternalForce="0,0,0") {
    angularDampingOverride=1
    angularSpringConstant=1500
    bAngularSpring=true
    bGravityOverrideInSimSpace=true
    bOverrideAngularDamping=true
    gravityScale=0.35
    simulationSpace=RootRelative
    alphaBoolBlend {blendOption=Linear}
    boundBone {boneName=bandFlap_b_l}
    physicsBodyDefinitions [{
      sphereCollisionRadius=0
      boundBone {boneName=bandFlap_b_l}
      boxExtents=<10 20 20>
      constraintSetup {angularLimitsMax=<0 20 30>, angularLimitsMin=<0 -20 -30>, angularTarget=<0 0 0>}
      localJointOffset=<5 0 0>
     }]
   }
   n17 = Anim Dynamics - Bone: bandFlap_a_r(Alpha=1, ExternalForce="0,0,0") {
    angularDampingOverride=1
    angularSpringConstant=1500
    bAngularSpring=true
    bGravityOverrideInSimSpace=true
    bOverrideAngularDamping=true
    gravityScale=0.35
    simulationSpace=RootRelative
    alphaBoolBlend {blendOption=Linear}
    boundBone {boneName=bandFlap_a_r}
    physicsBodyDefinitions [{
      sphereCollisionRadius=0
      boundBone {boneName=bandFlap_a_r}
      boxExtents=<10 20 20>
      constraintSetup {angularLimitsMax=<0 5 10>, angularLimitsMin=<0 -5 -10>, angularTarget=<0 0 0>}
      localJointOffset=<5 0 0>
     }]
   }
   Anim Dynamics - Bone: bandFlap_b_r(ComponentPose=@n17, Alpha=1, ExternalForce="0,0,0") {
    angularDampingOverride=1
    angularSpringConstant=1500
    bAngularSpring=true
    bGravityOverrideInSimSpace=true
    bOverrideAngularDamping=true
    gravityScale=0.35
    simulationSpace=RootRelative
    alphaBoolBlend {blendOption=Linear}
    boundBone {boneName=bandFlap_b_r}
    physicsBodyDefinitions [{
      sphereCollisionRadius=0
      boundBone {boneName=bandFlap_b_r}
      boxExtents=<10 20 20>
      constraintSetup {angularLimitsMax=<0 5 10>, angularLimitsMin=<0 -5 -10>, angularTarget=<0 0 0>}
      localJointOffset=<5 0 0>
     }]
   }
   n18 = Anim Dynamics - Bone: deault_backFlap_a(Alpha=0.8, ExternalForce="0,0,0") {
    angularDampingOverride=1
    angularSpringConstant=1500
    bAngularSpring=true
    bGravityOverrideInSimSpace=true
    bOverrideAngularDamping=true
    gravityScale=0.35
    simulationSpace=RootRelative
    alphaBoolBlend {blendOption=Linear}
    boundBone {boneName=deault_backFlap_a}
    componentLinearAccScale=<0.3 0.3 0.3>
    componentLinearVelScale=<0.3 0.3 0.3>
    physicsBodyDefinitions [{
      sphereCollisionRadius=0
      boundBone {boneName=deault_backFlap_a}
      boxExtents=<10 20 20>
      constraintSetup {angularLimitsMax=<0 5 0>, angularLimitsMin=<0 -5 -30>, angularTarget=<0 0 0>}
      localJointOffset=<5 0 0>
     }]
   }
   n19 = Anim Dynamics - Bone: default_backFlap_b(ComponentPose=@n18, Alpha=1, ExternalForce="0,0,0") {
    angularDampingOverride=1
    angularSpringConstant=1500
    bAngularSpring=true
    bGravityOverrideInSimSpace=true
    bOverrideAngularDamping=true
    gravityScale=0.35
    simulationSpace=RootRelative
    alphaBoolBlend {blendOption=Linear}
    boundBone {boneName=default_backFlap_b}
    componentLinearAccScale=<0.3 0.3 0.3>
    componentLinearVelScale=<0.3 0.3 0.3>
    physicsBodyDefinitions [{
      sphereCollisionRadius=0
      boundBone {boneName=default_backFlap_b}
      boxExtents=<10 20 20>
      constraintSetup {angularLimitsMax=<0 0 0>, angularLimitsMin=<0 0 -30>, angularTarget=<0 0 0>}
      localJointOffset=<5 0 0>
     }]
   }
   Anim Dynamics - Bone: default_backFlap_c(ComponentPose=@n19, Alpha=0.7, ExternalForce="0,0,0") {
    angularDampingOverride=1
    angularSpringConstant=1500
    bAngularSpring=true
    bGravityOverrideInSimSpace=true
    bOverrideAngularDamping=true
    gravityScale=0.35
    simulationSpace=RootRelative
    alphaBoolBlend {blendOption=Linear}
    boundBone {boneName=default_backFlap_c}
    componentLinearAccScale=<0.3 0.3 0.3>
    componentLinearVelScale=<0.3 0.3 0.3>
    physicsBodyDefinitions [{
      sphereCollisionRadius=0
      boundBone {boneName=default_backFlap_c}
      boxExtents=<10 20 20>
      constraintSetup {angularLimitsMax=<0 5 0>, angularLimitsMin=<0 -5 -30>, angularTarget=<0 0 0>}
      localJointOffset=<5 0 0>
     }]
   }
   n20 = Anim Dynamics - Bone: deault_backFlap_a(Alpha=0.7, ExternalForce="0,0,0") {
    angularDampingOverride=1
    angularSpringConstant=1500
    bAngularSpring=true
    bGravityOverrideInSimSpace=true
    bOverrideAngularDamping=true
    gravityScale=0.35
    simulationSpace=RootRelative
    alphaBoolBlend {blendOption=Linear}
    boundBone {boneName=deault_backFlap_a}
    componentLinearAccScale=<0.3 0.3 0.3>
    componentLinearVelScale=<0.3 0.3 0.3>
    physicsBodyDefinitions [{
      sphereCollisionRadius=0
      boundBone {boneName=deault_backFlap_a}
      boxExtents=<10 20 20>
      constraintSetup {angularLimitsMax=<0 5 0>, angularLimitsMin=<0 -5 -40>, angularTarget=<0 0 0>}
      localJointOffset=<5 0 0>
     }]
   }
   n21 = Anim Dynamics - Bone: default_backFlap_b(ComponentPose=@n20, Alpha=0.8, ExternalForce="0,0,0") {
    angularDampingOverride=1
    angularSpringConstant=1500
    bAngularSpring=true
    bGravityOverrideInSimSpace=true
    bOverrideAngularDamping=true
    gravityScale=0.35
    simulationSpace=RootRelative
    alphaBoolBlend {blendOption=Linear}
    boundBone {boneName=default_backFlap_b}
    componentLinearAccScale=<0.3 0.3 0.3>
    componentLinearVelScale=<0.3 0.3 0.3>
    physicsBodyDefinitions [{
      sphereCollisionRadius=0
      boundBone {boneName=default_backFlap_b}
      boxExtents=<10 20 20>
      constraintSetup {angularLimitsMax=<0 5 0>, angularLimitsMin=<0 -5 -40>, angularTarget=<0 0 0>}
      localJointOffset=<5 0 0>
     }]
   }
   n22 = Anim Dynamics - Bone: default_backFlap_c(ComponentPose=@n21, Alpha=0.9, ExternalForce="0,0,0") {
    angularDampingOverride=1
    angularSpringConstant=1500
    bAngularSpring=true
    bGravityOverrideInSimSpace=true
    bOverrideAngularDamping=true
    gravityScale=0.35
    simulationSpace=RootRelative
    alphaBoolBlend {blendOption=Linear}
    boundBone {boneName=default_backFlap_c}
    componentLinearAccScale=<0.35 0.35 0.35>
    componentLinearVelScale=<0.35 0.35 0.35>
    physicsBodyDefinitions [{
      sphereCollisionRadius=0
      boundBone {boneName=default_backFlap_c}
      boxExtents=<10 20 20>
      constraintSetup {angularLimitsMax=<0 5 0>, angularLimitsMin=<0 -5 -40>, angularTarget=<0 0 0>}
      localJointOffset=<5 0 0>
     }]
   }
   n23 = Anim Dynamics - Bone: default_backFlap_d(ComponentPose=@n22, Alpha=1, ExternalForce="0,0,0") {
    angularDampingOverride=1
    angularSpringConstant=1500
    bAngularSpring=true
    bGravityOverrideInSimSpace=true
    bOverrideAngularDamping=true
    gravityScale=0.35
    simulationSpace=RootRelative
    alphaBoolBlend {blendOption=Linear}
    boundBone {boneName=default_backFlap_d}
    componentLinearAccScale=<0.35 0.35 0.35>
    componentLinearVelScale=<0.35 0.35 0.35>
    physicsBodyDefinitions [{
      sphereCollisionRadius=0
      boundBone {boneName=default_backFlap_d}
      boxExtents=<10 20 20>
      constraintSetup {angularLimitsMax=<0 5 0>, angularLimitsMin=<0 -5 -40>, angularTarget=<0 0 0>}
      localJointOffset=<5 0 0>
     }]
   }
   n24 = Anim Dynamics - Bone: summer_shirts_bottom_fr(ComponentPose=@n23, Alpha=1, ExternalForce="0,0,0") {
    angularDampingOverride=1
    angularSpringConstant=1500
    bAngularSpring=true
    bGravityOverrideInSimSpace=true
    bOverrideAngularDamping=true
    bUseSphericalLimits=true
    gravityScale=0.35
    simulationSpace=RootRelative
    alphaBoolBlend {blendOption=Linear}
    boundBone {boneName=summer_shirts_bottom_fr}
    componentLinearAccScale=<0.4 0.4 0.4>
    componentLinearVelScale=<0.4 0.4 0.4>
    physicsBodyDefinitions [{
      collisionType=CustomSphere
      sphereCollisionRadius=4
      boundBone {boneName=summer_shirts_bottom_fr}
      boxExtents=<10 20 20>
      constraintSetup {angularLimitsMax=<0 0 20>, angularLimitsMin=<0 0 -5>, angularTarget=<0 0 0>}
      localJointOffset=<5 0 0>
     }]
    sphericalLimits [{limitRadius=6, drivingBone {boneName=spine_02}, sphereLocalOffset=<0 2 0>}]
   }
   n25 = Anim Dynamics - Bone: summer_shirts_bottom_bk(ComponentPose=@n24, Alpha=1, ExternalForce="0,0,0") {
    angularDampingOverride=1
    angularSpringConstant=1500
    bAngularSpring=true
    bGravityOverrideInSimSpace=true
    bOverrideAngularDamping=true
    bUseSphericalLimits=true
    gravityScale=0.35
    simulationSpace=RootRelative
    alphaBoolBlend {blendOption=Linear}
    boundBone {boneName=summer_shirts_bottom_bk}
    componentLinearAccScale=<0.4 0.4 0.4>
    componentLinearVelScale=<0.4 0.4 0.4>
    physicsBodyDefinitions [{
      collisionType=CustomSphere
      sphereCollisionRadius=3
      boundBone {boneName=summer_shirts_bottom_bk}
      boxExtents=<10 20 20>
      constraintSetup {angularLimitsMax=<0 0 40>, angularLimitsMin=<0 0 -5>, angularTarget=<0 0 0>}
      localJointOffset=<5 0 0>
     }]
    sphericalLimits [{limitRadius=6, drivingBone {boneName=spine_02}, sphereLocalOffset=<0 2 0>}]
   }
   n26 = Anim Dynamics - Bone: summer_shirts_bottom_fr_l(ComponentPose=@n25, Alpha=1, ExternalForce="0,0,0") {
    angularDampingOverride=1
    angularSpringConstant=1500
    bAngularSpring=true
    bGravityOverrideInSimSpace=true
    bOverrideAngularDamping=true
    bUseSphericalLimits=true
    gravityScale=0.35
    simulationSpace=RootRelative
    alphaBoolBlend {blendOption=Linear}
    boundBone {boneName=summer_shirts_bottom_fr_l}
    componentLinearAccScale=<0.4 0.4 0.4>
    componentLinearVelScale=<0.4 0.4 0.4>
    physicsBodyDefinitions [{
      collisionType=CustomSphere
      sphereCollisionRadius=5.5
      boundBone {boneName=summer_shirts_bottom_fr_l}
      boxExtents=<10 20 20>
      constraintSetup {angularLimitsMax=<0 0 20>, angularLimitsMin=<0 0 -5>, angularTarget=<0 0 0>}
      localJointOffset=<5 0 0>
     }]
    sphericalLimits [{limitRadius=6, drivingBone {boneName=spine_02}, sphereLocalOffset=<0 2 0>}]
   }
   n27 = Anim Dynamics - Bone: summer_shirts_side_l(ComponentPose=@n26, Alpha=1, ExternalForce="0,0,0") {
    angularDampingOverride=1
    angularSpringConstant=1500
    bAngularSpring=true
    bGravityOverrideInSimSpace=true
    bOverrideAngularDamping=true
    bUseSphericalLimits=true
    gravityScale=0.35
    simulationSpace=RootRelative
    alphaBoolBlend {blendOption=Linear}
    boundBone {boneName=summer_shirts_side_l}
    componentLinearAccScale=<0.4 0.4 0.4>
    componentLinearVelScale=<0.4 0.4 0.4>
    physicsBodyDefinitions [{
      collisionType=CustomSphere
      sphereCollisionRadius=7
      boundBone {boneName=summer_shirts_side_l}
      boxExtents=<10 20 20>
      constraintSetup {angularLimitsMax=<0 0 20>, angularLimitsMin=<0 0 -5>, angularTarget=<0 0 0>}
      localJointOffset=<5 0 0>
     }]
    sphericalLimits [{limitRadius=6, drivingBone {boneName=spine_02}, sphereLocalOffset=<0 2 0>}]
   }
   n28 = Anim Dynamics - Bone: summer_shirts_bottom_fr_r(ComponentPose=@n27, Alpha=1, ExternalForce="0,0,0") {
    angularDampingOverride=1
    angularSpringConstant=1500
    bAngularSpring=true
    bGravityOverrideInSimSpace=true
    bOverrideAngularDamping=true
    bUseSphericalLimits=true
    gravityScale=0.35
    simulationSpace=RootRelative
    alphaBoolBlend {blendOption=Linear}
    boundBone {boneName=summer_shirts_bottom_fr_r}
    componentLinearAccScale=<0.4 0.4 0.4>
    componentLinearVelScale=<0.4 0.4 0.4>
    physicsBodyDefinitions [{
      collisionType=CustomSphere
      sphereCollisionRadius=5.5
      boundBone {boneName=summer_shirts_bottom_fr_r}
      boxExtents=<10 20 20>
      constraintSetup {angularLimitsMax=<0 0 20>, angularLimitsMin=<0 0 -5>, angularTarget=<0 0 0>}
      localJointOffset=<-5 0 0>
     }]
    sphericalLimits [{limitRadius=6, drivingBone {boneName=spine_02}, sphereLocalOffset=<0 2 0>}]
   }
   Anim Dynamics - Bone: summer_shirts_side_r(ComponentPose=@n28, Alpha=1, ExternalForce="0,0,0") {
    angularDampingOverride=1
    angularSpringConstant=1500
    bAngularSpring=true
    bGravityOverrideInSimSpace=true
    bOverrideAngularDamping=true
    bUseSphericalLimits=true
    gravityScale=0.35
    simulationSpace=RootRelative
    alphaBoolBlend {blendOption=Linear}
    boundBone {boneName=summer_shirts_side_r}
    componentLinearAccScale=<0.4 0.4 0.4>
    componentLinearVelScale=<0.4 0.4 0.4>
    physicsBodyDefinitions [{
      collisionType=CustomSphere
      sphereCollisionRadius=7
      boundBone {boneName=summer_shirts_side_r}
      boxExtents=<10 20 20>
      constraintSetup {angularLimitsMax=<0 0 20>, angularLimitsMin=<0 0 -5>, angularTarget=<0 0 0>}
      localJointOffset=<-5 0 0>
     }]
    sphericalLimits [{limitRadius=6, drivingBone {boneName=spine_02}, sphereLocalOffset=<0 2 0>}]
   }
   n29 = "Transform (Modify) Bone - Bone: upperarm_twist_01_r"(
    ComponentPose=Local To Component()
    Alpha=Get pistol_ads_up_alpha_r() -> (pistol_ads_up_alpha_r:double)
    Translation="0,0,0"
    Rotation="-15,0,0"
    Scale="1,1,1"
   ) {
    rotationMode=BMM_Additive
    rotationSpace=BCS_BoneSpace
    scaleMode=BMM_Additive
    scaleSpace=BCS_BoneSpace
    translationMode=BMM_Additive
    translationSpace=BCS_BoneSpace
    alphaBoolBlend {blendOption=Linear}
    alphaScaleBiasClamp {inRange {max=0, min=-100}, outRange {max=0, min=1}}
    boneToModify {boneName=upperarm_twist_01_r}
    rotation {pitch=-15}
   }
   Component To Local(ComponentPose=@n29)
   "Transform (Modify) Bone - Bone: upperarm_twist_02_r"(Alpha=1, Translation="0,0,-5", Rotation="0,0,0", Scale="1,1,1") {
    rotationMode=BMM_Additive
    rotationSpace=BCS_BoneSpace
    scaleMode=BMM_Additive
    scaleSpace=BCS_BoneSpace
    translationMode=BMM_Additive
    translationSpace=BCS_BoneSpace
    alphaBoolBlend {blendOption=Linear}
    alphaScaleBiasClamp {inRange {max=0, min=-100}, outRange {max=0, min=1}}
    boneToModify {boneName=upperarm_twist_02_r}
    translation {z=-5}
   }
   RigidBody() {
    bClampLinearTranslationLimitToRefPose=true
    bOverrideWorldGravity=true
    overridePhysicsAsset=/Game/Characters/Player/Female/Medium/Bodies/F_MED_Believer_Summer/Meshes/F_MED_Believer_Summer_Physics
    simulationSpace=BaseBoneSpace
    alphaBoolBlend {blendOption=Linear}
    baseBoneRef {boneName=pelvis}
    componentLinearAccScale=<0.4 0.4 0.4>
    componentLinearVelScale=<0.4 0.4 0.4>
    overrideWorldGravity {z=-980}
   }
   n30 = Blend Poses by int!(
    BlendPose_0:Pose Link Structure=Use cached pose 'defaultVariantPose'()
    BlendPose_1:Pose Link Structure=Use cached pose 'summerVariantPose'()
    BlendPose_2:Pose Link Structure=Use cached pose 'JacketVariantPose'()
    BlendTime_0:float=0.1
    BlendTime_1:float=0.1
    BlendTime_2:float=0.1
    ActiveChildIndex:int=Get variantIndex() -> (variantIndex:int=0)
   ) -> (Pose:Pose Link Structure) {
    is=title AnimGraphNode_BlendListByInt
    blendType=Linear
    blendPose [{linkId=-1, sourceLinkId=-1}, {linkId=-1, sourceLinkId=-1}, {linkId=-1, sourceLinkId=-1}]
    blendTime [0.1,0.1,0.1]
   }
   Output Pose(Result=@n30) {layerGroup=None}
   n31 = Control Rig!(
    Source=Copy Pose From Mesh() {bUseAttachedParent=true}
    Alpha
    thigh_rot_l:double=Get thigh_rot_l()
    elbow_rot_l:double=Get elbow_rot_l() -> (elbow_rot_l:double)
    elbow_rot_r:double=Get elbow_rot_r() -> (elbow_rot_r:double)
    thigh_rot_r:double=Get thigh_rot_r()
    knee_rot_l:double=Get knee_rot_l() -> (knee_rot_l:double)
    knee_rot_r:double=Get knee_rot_r() -> (knee_rot_r:double)
    psd_arm_up_alpha_l:double=Get psd_arm_up_alpha_l() -> (psd_arm_up_alpha_l:double)
    psd_arm_up_alpha_r:double=Get psd_arm_up_alpha_r() -> (psd_arm_up_alpha_r:double)
    psd_arm_fwd_alpha_l:double=Get psd_arm_fwd_alpha_l() -> (psd_arm_fwd_alpha_l:double)
    psd_arm_fwd_alpha_r:double=Get psd_arm_fwd_alpha_r() -> (psd_arm_fwd_alpha_r:double)
    psd_arm_bk_alpha_l:double=Get psd_arm_bk_alpha_l() -> (psd_arm_bk_alpha_l:double)
    psd_arm_bk_alpha_r:double=Get psd_arm_bk_alpha_r() -> (psd_arm_bk_alpha_r:double)
   ) -> (Pose) {
    alphaBoolBlend {blendOption=Linear}
    controlRigAssetReference {blueprintRigClass=&/Muscle_Volume_ControlRig.Muscle_Volume_ControlRig_C}
   }
   Save cached pose 'musclePose'(Pose=@n31) {cacheName=musclePose}
   n32 = "Constraint - zipper_a_base (PPPPP)"!(
    ComponentPose=Local To Component(LocalPose=Use cached pose 'musclePose'())
    Alpha=1
    ConstraintWeights_0=0.04
    ConstraintWeights_1:float=0.44
    ConstraintWeights_2:float=0.06
    ConstraintWeights_3:float=0.12
    ConstraintWeights_4:float=0.34
   ) -> (Pose) {
    alphaBoolBlend {blendOption=Linear}
    boneToModify {boneName=zipper_a_base}
    constraintSetup [
     {targetBone {boneName=spine_01}}
     {targetBone {boneName=spine_02}}
     {targetBone {boneName=spine_03}}
     {targetBone {boneName=spine_04}}
     {targetBone {boneName=pelvis}}
    ]
    constraintWeights [1,1,1,1,1]
   }
   n33 = "Constraint - zipper_b_base (PPPPP)"!(
    ComponentPose=@n32
    Alpha=1
    ConstraintWeights_0=0.04
    ConstraintWeights_1:float=0.04
    ConstraintWeights_2:float=0.73
    ConstraintWeights_3:float=0.08
    ConstraintWeights_4:float=0.11
   ) -> (Pose) {
    alphaBoolBlend {blendOption=Linear}
    boneToModify {boneName=zipper_b_base}
    constraintSetup [
     {targetBone {boneName=spine_02}}
     {targetBone {boneName=spine_03}}
     {targetBone {boneName=spine_04}}
     {targetBone {boneName=spine_05}}
     {targetBone {boneName=pec_volume_r}}
    ]
    constraintWeights [1,1,1,1,1]
   }
   n34 = "Constraint - zipper_c_base (PPPPPP)"!(
    ComponentPose=@n33
    Alpha=1
    ConstraintWeights_0=0.24
    ConstraintWeights_1:float=0.2
    ConstraintWeights_2:float=0.2
    ConstraintWeights_3:float=0.3
    ConstraintWeights_4:float=0.03
    ConstraintWeights_5:float=0.03
   ) -> (Pose) {
    alphaBoolBlend {blendOption=Linear}
    boneToModify {boneName=zipper_c_base}
    constraintSetup [
     {targetBone {boneName=clavicle_l}}
     {targetBone {boneName=pec_volume_l}}
     {targetBone {boneName=spine_04}}
     {targetBone {boneName=spine_05}}
     {targetBone {boneName=upperarm_twist_01_l}}
     {targetBone {boneName=upperarm_fr_volume_l}}
    ]
    constraintWeights [1,1,1,1,1,1]
   }
   n35 = "Constraint - zipper_d_base (PPPPP)"!(
    ComponentPose=@n34
    Alpha=1
    ConstraintWeights_0=0.04
    ConstraintWeights_1:float=0.44
    ConstraintWeights_2:float=0.05
    ConstraintWeights_3:float=0.12
    ConstraintWeights_4:float=0.35
   ) -> (Pose) {
    alphaBoolBlend {blendOption=Linear}
    boneToModify {boneName=zipper_d_base}
    constraintSetup [
     {targetBone {boneName=spine_01}}
     {targetBone {boneName=spine_02}}
     {targetBone {boneName=spine_03}}
     {targetBone {boneName=spine_04}}
     {targetBone {boneName=pelvis}}
    ]
    constraintWeights [1,1,1,1,1]
   }
   n36 = Anim Dynamics - Bone: zipper_a(ComponentPose=@n35, Alpha=1, ExternalForce="0,0,0") {
    angularDampingOverride=1
    angularSpringConstant=1500
    bAngularSpring=true
    bGravityOverrideInSimSpace=true
    bOverrideAngularDamping=true
    gravityScale=0.35
    simulationSpace=RootRelative
    alphaBoolBlend {blendOption=Linear}
    boundBone {boneName=zipper_a}
    componentLinearAccScale=<0.25 0.25 0.25>
    componentLinearVelScale=<0.25 0.25 0.25>
    physicsBodyDefinitions [{
      sphereCollisionRadius=0
      boundBone {boneName=zipper_a}
      boxExtents=<5 10 10>
      constraintSetup {angularLimitsMax=<5 0 5>, angularLimitsMin=<-5 -40 -5>, angularTarget=<0 0 0>}
      localJointOffset=<5 0 0>
     }]
   }
   n37 = Anim Dynamics - Bone: zipper_b(ComponentPose=@n36, Alpha=1, ExternalForce="0,0,0") {
    angularDampingOverride=1
    angularSpringConstant=1500
    bAngularSpring=true
    bGravityOverrideInSimSpace=true
    bOverrideAngularDamping=true
    gravityScale=0.35
    simulationSpace=RootRelative
    alphaBoolBlend {blendOption=Linear}
    boundBone {boneName=zipper_b}
    componentLinearAccScale=<0.25 0.25 0.25>
    componentLinearVelScale=<0.25 0.25 0.25>
    physicsBodyDefinitions [{
      sphereCollisionRadius=0
      boundBone {boneName=zipper_b}
      boxExtents=<5 10 10>
      constraintSetup {angularLimitsMax=<5 5 40>, angularLimitsMin=<-5 -5 0>, angularTarget=<0 0 0>}
      localJointOffset=<5 0 0>
     }]
   }
   n38 = Anim Dynamics - Bone: zipper_c(ComponentPose=@n37, Alpha=1, ExternalForce="0,0,0") {
    angularDampingOverride=1
    angularSpringConstant=1500
    bAngularSpring=true
    bGravityOverrideInSimSpace=true
    bOverrideAngularDamping=true
    gravityScale=0.35
    simulationSpace=RootRelative
    alphaBoolBlend {blendOption=Linear}
    boundBone {boneName=zipper_c}
    componentLinearAccScale=<0.25 0.25 0.25>
    componentLinearVelScale=<0.25 0.25 0.25>
    physicsBodyDefinitions [{
      sphereCollisionRadius=0
      boundBone {boneName=zipper_c}
      boxExtents=<5 10 10>
      constraintSetup {angularLimitsMax=<5 5 40>, angularLimitsMin=<-5 -5 0>, angularTarget=<0 0 0>}
      localJointOffset=<5 0 0>
     }]
   }
   n39 = Anim Dynamics - Bone: zipper_d(ComponentPose=@n38, Alpha=1, ExternalForce="0,0,0") {
    angularDampingOverride=1
    angularSpringConstant=1500
    bAngularSpring=true
    bGravityOverrideInSimSpace=true
    bOverrideAngularDamping=true
    gravityScale=0.35
    simulationSpace=RootRelative
    alphaBoolBlend {blendOption=Linear}
    boundBone {boneName=zipper_d}
    componentLinearAccScale=<0.25 0.25 0.25>
    componentLinearVelScale=<0.25 0.25 0.25>
    physicsBodyDefinitions [{
      sphereCollisionRadius=0
      boundBone {boneName=zipper_d}
      boxExtents=<5 10 10>
      constraintSetup {angularLimitsMax=<5 5 40>, angularLimitsMin=<-5 -5 -5>, angularTarget=<0 0 0>}
      localJointOffset=<5 0 0>
     }]
   }
   n40 = "Transform (Modify) Bone - Bone: glute_volume_l"(ComponentPose=@n39, Alpha=Get thigh_rot_l(), Translation="2.5,-6,0", Rotation="0,0,0", Scale="1,1.2,1.2") {
    rotationMode=BMM_Additive
    rotationSpace=BCS_BoneSpace
    scaleMode=BMM_Additive
    scaleSpace=BCS_BoneSpace
    translationMode=BMM_Additive
    translationSpace=BCS_BoneSpace
    alphaBoolBlend {blendOption=Linear}
    alphaScaleBiasClamp {bMapRange=true, inRange {max=0, min=-100}, outRange {max=0, min=1}}
    boneToModify {boneName=glute_volume_l}
    scale {y=1.2, z=1.2}
    translation=<2.5 -6>
   }
   n41 = "Transform (Modify) Bone - Bone: glute_volume_r"(ComponentPose=@n40, Alpha=Get thigh_rot_r(), Translation="2.5,6,0", Rotation="0,0,0", Scale="1,1.2,1.2") {
    rotationMode=BMM_Additive
    rotationSpace=BCS_BoneSpace
    scaleMode=BMM_Additive
    scaleSpace=BCS_BoneSpace
    translationMode=BMM_Additive
    translationSpace=BCS_BoneSpace
    alphaBoolBlend {blendOption=Linear}
    alphaScaleBiasClamp {bMapRange=true, inRange {max=0, min=-100}, outRange {max=0, min=1}}
    boneToModify {boneName=glute_volume_r}
    scale {y=1.2, z=1.2}
    translation=<2.5 6>
   }
   n42 = Anim Dynamics - Bone: belt(ComponentPose=@n41, Alpha=1, ExternalForce="0,0,0") {
    angularDampingOverride=1
    angularSpringConstant=1500
    bAngularSpring=true
    bGravityOverrideInSimSpace=true
    bOverrideAngularDamping=true
    gravityScale=0.35
    simulationSpace=RootRelative
    alphaBoolBlend {blendOption=Linear}
    boundBone {boneName=belt}
    physicsBodyDefinitions [{
      sphereCollisionRadius=0
      boundBone {boneName=belt}
      boxExtents=<5 15 15>
      constraintSetup {angularLimitsMax=<0 0 10>, angularLimitsMin=<0 0 -10>, angularTarget=<0 0 0>}
      localJointOffset=<5 0 0>
     }]
   }
   Save cached pose 'JacketVariantPose'(Pose=Component To Local(ComponentPose=@n42)) {cacheName=JacketVariantPose}
   n43 = "Constraint - default_necklace_jewel (PPPPP)"!(
    ComponentPose=Local To Component(LocalPose=Use cached pose 'musclePose'())
    Alpha=1
    ConstraintWeights_0=0.28
    ConstraintWeights_1:float=0.05
    ConstraintWeights_2:float=0.07
    ConstraintWeights_3:float=0.21
    ConstraintWeights_4:float=0.39
   ) -> (Pose) {
    alphaBoolBlend {blendOption=Linear}
    boneToModify {boneName=default_necklace_jewel}
    constraintSetup [
     {targetBone {boneName=neck_02}}
     {targetBone {boneName=head}}
     {targetBone {boneName=spine_04}}
     {targetBone {boneName=neck_01}}
     {targetBone {boneName=spine_05}}
    ]
    constraintWeights [1,1,1,1,1]
   }
   n44 = "Constraint - deault_backFlap_a_base (PPPP)"!(
    ComponentPose=@n43
    Alpha=1
    ConstraintWeights_0=0.1
    ConstraintWeights_1:float=0.13
    ConstraintWeights_2:float=0.67
    ConstraintWeights_3:float=0.1
   ) -> (Pose) {
    alphaBoolBlend {blendOption=Linear}
    boneToModify {boneName=deault_backFlap_a_base}
    constraintSetup [{targetBone {boneName=clavicle_r}}, {targetBone {boneName=spine_04}}, {targetBone {boneName=spine_05}}, {targetBone {boneName=clavicle_l}}]
    constraintWeights [1,1,1,1]
   }
   n45 = "Constraint - default_backFlap_b_base (PPPP)"!(
    ComponentPose=@n44
    Alpha=1
    ConstraintWeights_0=0.05
    ConstraintWeights_1:float=0.3
    ConstraintWeights_2:float=0.6
    ConstraintWeights_3:float=0.05
   ) -> (Pose) {
    alphaBoolBlend {blendOption=Linear}
    boneToModify {boneName=default_backFlap_b_base}
    constraintSetup [{targetBone {boneName=clavicle_r}}, {targetBone {boneName=spine_04}}, {targetBone {boneName=spine_05}}, {targetBone {boneName=clavicle_l}}]
    constraintWeights [1,1,1,1]
   }
   n46 = "Constraint - default_backFlap_c_base (PPPPPPP)"!(
    ComponentPose=@n45
    Alpha=1
    ConstraintWeights_0=0.02
    ConstraintWeights_1:float=0.05
    ConstraintWeights_2:float=0.15
    ConstraintWeights_3:float=0.15
    ConstraintWeights_4:float=0.35
    ConstraintWeights_5:float=0.26
    ConstraintWeights_6:float=0.02
   ) -> (Pose) {
    alphaBoolBlend {blendOption=Linear}
    boneToModify {boneName=default_backFlap_c_base}
    constraintSetup [
     {targetBone {boneName=clavicle_l}}
     {targetBone {boneName=spine_01}}
     {targetBone {boneName=spine_02}}
     {targetBone {boneName=spine_03}}
     {targetBone {boneName=spine_04}}
     {targetBone {boneName=spine_05}}
     {targetBone {boneName=clavicle_r}}
    ]
    constraintWeights [1,1,1,1,1,1,1]
   }
   n47 = "Constraint - default_bottom_fr_a_base (PPPPP)"!(
    ComponentPose=@n46
    Alpha=1
    ConstraintWeights_0=0.03
    ConstraintWeights_1:float=0.24
    ConstraintWeights_2:float=0.13
    ConstraintWeights_3:float=0.57
    ConstraintWeights_4:float=0.03
   ) -> (Pose) {
    alphaBoolBlend {blendOption=Linear}
    boneToModify {boneName=default_bottom_fr_a_base}
    constraintSetup [
     {targetBone {boneName=spine_01}}
     {targetBone {boneName=spine_02}}
     {targetBone {boneName=spine_03}}
     {targetBone {boneName=spine_04}}
     {targetBone {boneName=pelvis}}
    ]
    constraintWeights [1,1,1,1,1]
   }
   n48 = "Constraint - default_bottom_fr_b (PPPPP)"!(
    ComponentPose=@n47
    Alpha=1
    ConstraintWeights_0=0.04
    ConstraintWeights_1:float=0.41
    ConstraintWeights_2:float=0.1
    ConstraintWeights_3:float=0.3
    ConstraintWeights_4:float=0.15
   ) -> (Pose) {
    alphaBoolBlend {blendOption=Linear}
    boneToModify {boneName=default_bottom_fr_b}
    constraintSetup [
     {targetBone {boneName=spine_01}}
     {targetBone {boneName=spine_02}}
     {targetBone {boneName=spine_03}}
     {targetBone {boneName=spine_04}}
     {targetBone {boneName=pelvis}}
    ]
    constraintWeights [1,1,1,1,1]
   }
   n49 = "Constraint - default_bottom_bk_a_base (PPPPPP)"!(
    ComponentPose=@n48
    Alpha=1
    ConstraintWeights_0=0.1
    ConstraintWeights_1:float=0.4
    ConstraintWeights_2:float=0.16
    ConstraintWeights_3:float=0.23
    ConstraintWeights_4:float=0.04
    ConstraintWeights_5:float=0.07
   ) -> (Pose) {
    alphaBoolBlend {blendOption=Linear}
    boneToModify {boneName=default_bottom_bk_a_base}
    constraintSetup [
     {targetBone {boneName=spine_01}}
     {targetBone {boneName=spine_02}}
     {targetBone {boneName=spine_03}}
     {targetBone {boneName=spine_04}}
     {targetBone {boneName=spine_05}}
     {targetBone {boneName=pelvis}}
    ]
    constraintWeights [1,1,1,1,1,1]
   }
   n50 = "Constraint - default_bottom_bk_b (PPPPP)"!(
    ComponentPose=@n49
    Alpha=1
    ConstraintWeights_0=0.11
    ConstraintWeights_1:float=0.46
    ConstraintWeights_2:float=0.1
    ConstraintWeights_3:float=0.1
    ConstraintWeights_4:float=0.23
   ) -> (Pose) {
    alphaBoolBlend {blendOption=Linear}
    boneToModify {boneName=default_bottom_bk_b}
    constraintSetup [
     {targetBone {boneName=spine_01}}
     {targetBone {boneName=spine_02}}
     {targetBone {boneName=spine_03}}
     {targetBone {boneName=spine_04}}
     {targetBone {boneName=pelvis}}
    ]
    constraintWeights [1,1,1,1,1]
   }
   n51 = "Constraint - default_bottom_fr_a_base_l (PPPPP)"!(
    ComponentPose=@n50
    Alpha=1
    ConstraintWeights_0=0.04
    ConstraintWeights_1:float=0.23
    ConstraintWeights_2:float=0.12
    ConstraintWeights_3:float=0.58
    ConstraintWeights_4:float=0.03
   ) -> (Pose) {
    alphaBoolBlend {blendOption=Linear}
    boneToModify {boneName=default_bottom_fr_a_base_l}
    constraintSetup [
     {targetBone {boneName=spine_01}}
     {targetBone {boneName=spine_02}}
     {targetBone {boneName=spine_03}}
     {targetBone {boneName=spine_04}}
     {targetBone {boneName=pelvis}}
    ]
    constraintWeights [1,1,1,1,1]
   }
   n52 = "Constraint - default_bottom_fr_b_l (PPPPP)"!(
    ComponentPose=@n51
    Alpha=1
    ConstraintWeights_0=0.04
    ConstraintWeights_1:float=0.44
    ConstraintWeights_2:float=0.1
    ConstraintWeights_3:float=0.27
    ConstraintWeights_4:float=0.15
   ) -> (Pose) {
    alphaBoolBlend {blendOption=Linear}
    boneToModify {boneName=default_bottom_fr_b_l}
    constraintSetup [
     {targetBone {boneName=spine_01}}
     {targetBone {boneName=spine_02}}
     {targetBone {boneName=spine_03}}
     {targetBone {boneName=spine_04}}
     {targetBone {boneName=pelvis}}
    ]
    constraintWeights [1,1,1,1,1]
   }
   n53 = "Constraint - default_bottom_side_a_base_l (PPPPPPP)"!(
    ComponentPose=@n52
    Alpha=1
    ConstraintWeights_0=0.03
    ConstraintWeights_1:float=0.05
    ConstraintWeights_2:float=0.29
    ConstraintWeights_3:float=0.13
    ConstraintWeights_4:float=0.44
    ConstraintWeights_5:float=0.03
    ConstraintWeights_6:float=0.03
   ) -> (Pose) {
    alphaBoolBlend {blendOption=Linear}
    boneToModify {boneName=default_bottom_side_a_base_l}
    constraintSetup [
     {targetBone {boneName=clavicle_l}}
     {targetBone {boneName=spine_01}}
     {targetBone {boneName=spine_02}}
     {targetBone {boneName=spine_03}}
     {targetBone {boneName=spine_04}}
     {targetBone {boneName=spine_05}}
     {targetBone {boneName=pelvis}}
    ]
    constraintWeights [1,1,1,1,1,1,1]
   }
   n54 = "Constraint - default_bottom_side_b_l (PPPPP)"!(
    ComponentPose=@n53
    Alpha=1
    ConstraintWeights_0=0.06
    ConstraintWeights_1:float=0.49
    ConstraintWeights_2:float=0.1
    ConstraintWeights_3:float=0.21
    ConstraintWeights_4:float=0.14
   ) -> (Pose) {
    alphaBoolBlend {blendOption=Linear}
    boneToModify {boneName=default_bottom_side_b_l}
    constraintSetup [
     {targetBone {boneName=spine_01}}
     {targetBone {boneName=spine_02}}
     {targetBone {boneName=spine_03}}
     {targetBone {boneName=spine_04}}
     {targetBone {boneName=pelvis}}
    ]
    constraintWeights [1,1,1,1,1]
   }
   n55 = "Constraint - default_bottom_bk_a_base_l (PPPPPPP)"!(
    ComponentPose=@n54
    Alpha=1
    ConstraintWeights_0=0.05
    ConstraintWeights_1:float=0.07
    ConstraintWeights_2:float=0.3
    ConstraintWeights_3:float=0.14
    ConstraintWeights_4:float=0.34
    ConstraintWeights_5:float=0.05
    ConstraintWeights_6:float=0.05
   ) -> (Pose) {
    alphaBoolBlend {blendOption=Linear}
    boneToModify {boneName=default_bottom_bk_a_base_l}
    constraintSetup [
     {targetBone {boneName=clavicle_l}}
     {targetBone {boneName=spine_01}}
     {targetBone {boneName=spine_02}}
     {targetBone {boneName=spine_03}}
     {targetBone {boneName=spine_04}}
     {targetBone {boneName=spine_05}}
     {targetBone {boneName=pelvis}}
    ]
    constraintWeights [1,1,1,1,1,1,1]
   }
   n56 = "Constraint - default_bottom_bk_b_l (PPPPP)"!(
    ComponentPose=@n55
    Alpha=1
    ConstraintWeights_0=0.1
    ConstraintWeights_1:float=0.48
    ConstraintWeights_2:float=0.11
    ConstraintWeights_3:float=0.16
    ConstraintWeights_4:float=0.15
   ) -> (Pose) {
    alphaBoolBlend {blendOption=Linear}
    boneToModify {boneName=default_bottom_bk_b_l}
    constraintSetup [
     {targetBone {boneName=spine_01}}
     {targetBone {boneName=spine_02}}
     {targetBone {boneName=spine_03}}
     {targetBone {boneName=spine_04}}
     {targetBone {boneName=pelvis}}
    ]
    constraintWeights [1,1,1,1,1]
   }
   n57 = "Constraint - default_bottom_fr_a_base_r (PPPPP)"!(
    ComponentPose=@n56
    Alpha=1
    ConstraintWeights_0=0.04
    ConstraintWeights_1:float=0.23
    ConstraintWeights_2:float=0.12
    ConstraintWeights_3:float=0.58
    ConstraintWeights_4:float=0.03
   ) -> (Pose) {
    alphaBoolBlend {blendOption=Linear}
    boneToModify {boneName=default_bottom_fr_a_base_r}
    constraintSetup [
     {targetBone {boneName=spine_01}}
     {targetBone {boneName=spine_02}}
     {targetBone {boneName=spine_03}}
     {targetBone {boneName=spine_04}}
     {targetBone {boneName=pelvis}}
    ]
    constraintWeights [1,1,1,1,1]
   }
   n58 = "Constraint - default_bottom_fr_b_r (PPPPP)"!(
    ComponentPose=@n57
    Alpha=1
    ConstraintWeights_0=0.04
    ConstraintWeights_1:float=0.44
    ConstraintWeights_2:float=0.1
    ConstraintWeights_3:float=0.27
    ConstraintWeights_4:float=0.15
   ) -> (Pose) {
    alphaBoolBlend {blendOption=Linear}
    boneToModify {boneName=default_bottom_fr_b_r}
    constraintSetup [
     {targetBone {boneName=spine_01}}
     {targetBone {boneName=spine_02}}
     {targetBone {boneName=spine_03}}
     {targetBone {boneName=spine_04}}
     {targetBone {boneName=pelvis}}
    ]
    constraintWeights [1,1,1,1,1]
   }
   n59 = "Constraint - default_bottom_side_a_base_r (PPPPPPP)"!(
    ComponentPose=@n58
    Alpha=1
    ConstraintWeights_0=0.05
    ConstraintWeights_1:float=0.29
    ConstraintWeights_2:float=0.13
    ConstraintWeights_3:float=0.44
    ConstraintWeights_4:float=0.03
    ConstraintWeights_5:float=0.03
    ConstraintWeights_6:float=0.03
   ) -> (Pose) {
    alphaBoolBlend {blendOption=Linear}
    boneToModify {boneName=default_bottom_side_a_base_r}
    constraintSetup [
     {targetBone {boneName=spine_01}}
     {targetBone {boneName=spine_02}}
     {targetBone {boneName=spine_03}}
     {targetBone {boneName=spine_04}}
     {targetBone {boneName=spine_05}}
     {targetBone {boneName=clavicle_r}}
     {targetBone {boneName=pelvis}}
    ]
    constraintWeights [1,1,1,1,1,1,1]
   }
   n60 = "Constraint - default_bottom_side_b_r (PPPPP)"!(
    ComponentPose=@n59
    Alpha=1
    ConstraintWeights_0=0.06
    ConstraintWeights_1:float=0.49
    ConstraintWeights_2:float=0.1
    ConstraintWeights_3:float=0.21
    ConstraintWeights_4:float=0.14
   ) -> (Pose) {
    alphaBoolBlend {blendOption=Linear}
    boneToModify {boneName=default_bottom_side_b_r}
    constraintSetup [
     {targetBone {boneName=spine_01}}
     {targetBone {boneName=spine_02}}
     {targetBone {boneName=spine_03}}
     {targetBone {boneName=spine_04}}
     {targetBone {boneName=pelvis}}
    ]
    constraintWeights [1,1,1,1,1]
   }
   n61 = "Constraint - default_bottom_bk_a_base_r (PPPPPPP)"!(
    ComponentPose=@n60
    Alpha=1
    ConstraintWeights_0=0.07
    ConstraintWeights_1:float=0.3
    ConstraintWeights_2:float=0.14
    ConstraintWeights_3:float=0.34
    ConstraintWeights_4:float=0.05
    ConstraintWeights_5:float=0.05
    ConstraintWeights_6:float=0.05
   ) -> (Pose) {
    alphaBoolBlend {blendOption=Linear}
    boneToModify {boneName=default_bottom_bk_a_base_r}
    constraintSetup [
     {targetBone {boneName=spine_01}}
     {targetBone {boneName=spine_02}}
     {targetBone {boneName=spine_03}}
     {targetBone {boneName=spine_04}}
     {targetBone {boneName=spine_05}}
     {targetBone {boneName=clavicle_r}}
     {targetBone {boneName=pelvis}}
    ]
    constraintWeights [1,1,1,1,1,1,1]
   }
   n62 = "Constraint - default_bottom_bk_b_r (PPPPP)"!(
    ComponentPose=@n61
    Alpha=1
    ConstraintWeights_0=0.1
    ConstraintWeights_1:float=0.48
    ConstraintWeights_2:float=0.11
    ConstraintWeights_3:float=0.16
    ConstraintWeights_4:float=0.15
   ) -> (Pose) {
    alphaBoolBlend {blendOption=Linear}
    boneToModify {boneName=default_bottom_bk_b_r}
    constraintSetup [
     {targetBone {boneName=spine_01}}
     {targetBone {boneName=spine_02}}
     {targetBone {boneName=spine_03}}
     {targetBone {boneName=spine_04}}
     {targetBone {boneName=pelvis}}
    ]
    constraintWeights [1,1,1,1,1]
   }
   n63 = "Transform (Modify) Bone - Bone: glute_volume_l"(ComponentPose=@n62, Alpha=Get thigh_rot_l(), Translation="-3,2,0", Rotation="0,0,0", Scale="1,1,1") {
    rotationMode=BMM_Additive
    rotationSpace=BCS_BoneSpace
    scaleMode=BMM_Additive
    scaleSpace=BCS_BoneSpace
    translationMode=BMM_Additive
    translationSpace=BCS_BoneSpace
    alphaBoolBlend {blendOption=Linear}
    alphaScaleBiasClamp {bMapRange=true, inRange {max=0, min=-100}, outRange {max=0, min=1}}
    boneToModify {boneName=glute_volume_l}
    translation=<-3 2>
   }
   n64 = "Transform (Modify) Bone - Bone: glute_volume_r"(ComponentPose=@n63, Alpha=Get thigh_rot_r(), Translation="3,-2,0", Rotation="0,0,0", Scale="1,1,1") {
    rotationMode=BMM_Additive
    rotationSpace=BCS_BoneSpace
    scaleMode=BMM_Additive
    scaleSpace=BCS_BoneSpace
    translationMode=BMM_Additive
    translationSpace=BCS_BoneSpace
    alphaBoolBlend {blendOption=Linear}
    alphaScaleBiasClamp {bMapRange=true, inRange {max=0, min=-100}, outRange {max=0, min=1}}
    boneToModify {boneName=glute_volume_r}
    translation=<3 -2>
   }
   n65 = RigidBody(ComponentPose=@n64) {
    bClampLinearTranslationLimitToRefPose=true
    bOverrideWorldGravity=true
    overridePhysicsAsset=&/F_MED_Believer_ShirtsBottom_Physics
    simulationSpace=BaseBoneSpace
    alphaBoolBlend {blendOption=Linear}
    baseBoneRef {boneName=pelvis}
    componentLinearAccScale=<0.3 0.3 0.3>
    componentLinearVelScale=<0.3 0.3 0.3>
    overrideWorldGravity {z=-980}
   }
   n66 = RigidBody(ComponentPose=@n65) {
    bClampLinearTranslationLimitToRefPose=true
    bOverrideWorldGravity=true
    overridePhysicsAsset=&/F_MED_Believer_ShirtsBack_Physics
    simulationSpace=BaseBoneSpace
    alphaBoolBlend {blendOption=Linear}
    baseBoneRef {boneName=spine_02}
    componentLinearAccScale=<0.3 0.3 0.3>
    componentLinearVelScale=<0.3 0.3 0.3>
    overrideWorldGravity {z=-980}
   }
   Save cached pose 'defaultVariantPose'(Pose=Component To Local(ComponentPose=@n66)) {cacheName=defaultVariantPose}
   n67 = "Constraint - pantsButton_b_base (PPP)"!(
    ComponentPose=Local To Component(LocalPose=Use cached pose 'musclePose'())
    Alpha=1
    ConstraintWeights_0=0.98
    ConstraintWeights_1:float=0.01
    ConstraintWeights_2:float=0.01
   ) -> (Pose) {
    alphaBoolBlend {blendOption=Linear}
    boneToModify {boneName=pantsButton_b_base}
    constraintSetup [{targetBone {boneName=pelvis}}, {targetBone {boneName=crouch_volume_r}}, {targetBone {boneName=thigh_twist_01_r}}]
    constraintWeights [1,1,1]
   }
   n68 = "Constraint - pantsButton_c_base (PPPPP)"!(
    ComponentPose=@n67
    Alpha=1
    ConstraintWeights_0=0.85
    ConstraintWeights_1:float=0.03
    ConstraintWeights_2:float=0.05
    ConstraintWeights_3:float=0.05
    ConstraintWeights_4:float=0.02
   ) -> (Pose) {
    alphaBoolBlend {blendOption=Linear}
    boneToModify {boneName=pantsButton_c_base}
    constraintSetup [
     {targetBone {boneName=pelvis}}
     {targetBone {boneName=thigh_twist_01_l}}
     {targetBone {boneName=thigh_twist_01_r}}
     {targetBone {boneName=crouch_volume_r}}
     {targetBone {boneName=crouch_volume_l}}
    ]
    constraintWeights [1,1,1,1,1]
   }
   n69 = "Constraint - pantsButton_d_base (PPPPP)"!(
    ComponentPose=@n68
    Alpha=1
    ConstraintWeights_0=0.71
    ConstraintWeights_1:float=0.06
    ConstraintWeights_2:float=0.12
    ConstraintWeights_3:float=0.07
    ConstraintWeights_4:float=0.04
   ) -> (Pose) {
    alphaBoolBlend {blendOption=Linear}
    boneToModify {boneName=pantsButton_d_base}
    constraintSetup [
     {targetBone {boneName=pelvis}}
     {targetBone {boneName=thigh_twist_01_l}}
     {targetBone {boneName=thigh_twist_01_r}}
     {targetBone {boneName=crouch_volume_r}}
     {targetBone {boneName=crouch_volume_l}}
    ]
    constraintWeights [1,1,1,1,1]
   }
   n70 = "Constraint - bandFlap_a_base_l (PPP)"!(ComponentPose=@n69, Alpha=1, ConstraintWeights_0=0.1, ConstraintWeights_1:float=0.1, ConstraintWeights_2:float=0.8) -> (Pose) {
    alphaBoolBlend {blendOption=Linear}
    boneToModify {boneName=bandFlap_a_base_l}
    constraintSetup [{targetBone {boneName=upperarm_twist_02_l}}, {targetBone {boneName=upperarm_l}}, {targetBone {boneName=upperarm_twist_01_l}}]
    constraintWeights [1,1,1]
   }
   n71 = "Constraint - bandFlap_b_base_l (PP)"!(ComponentPose=@n70, Alpha=1, ConstraintWeights_0=0.1, ConstraintWeights_1:float=0.9) -> (Pose) {
    alphaBoolBlend {blendOption=Linear}
    boneToModify {boneName=bandFlap_b_base_l}
    constraintSetup [{targetBone {boneName=upperarm_twist_02_l}}, {targetBone {boneName=upperarm_twist_01_l}}]
    constraintWeights [1,1]
   }
   n72 = "Constraint - bandFlap_a_base_r (PP)"!(ComponentPose=@n71, Alpha=1, ConstraintWeights_0=0.7, ConstraintWeights_1:float=0.3) -> (Pose) {
    alphaBoolBlend {blendOption=Linear}
    boneToModify {boneName=bandFlap_a_base_r}
    constraintSetup [{targetBone {boneName=upperarm_twist_01_r}}, {targetBone {boneName=bicep_out_volume_r}}]
    constraintWeights [1,1]
   }
   n73 = "Constraint - bandFlap_b_base_r (PPP)"!(ComponentPose=@n72, Alpha=1, ConstraintWeights_0=0.6, ConstraintWeights_1:float=0.3, ConstraintWeights_2:float=0.1) -> (Pose) {
    alphaBoolBlend {blendOption=Linear}
    boneToModify {boneName=bandFlap_b_base_r}
    constraintSetup [{targetBone {boneName=upperarm_twist_01_r}}, {targetBone {boneName=bicep_out_volume_r}}, {targetBone {boneName=upperarm_twist_02_r}}]
    constraintWeights [1,1,1]
   }
   n74 = "Constraint - deault_backFlap_a_base (PPPP)"(ComponentPose=@n73, Alpha=1, ConstraintWeights_0=0.1, ConstraintWeights_1=0.13, ConstraintWeights_2=0.67, ConstraintWeights_3=0.1) {
    alphaBoolBlend {blendOption=Linear}
    boneToModify {boneName=deault_backFlap_a_base}
    constraintSetup [{targetBone {boneName=clavicle_r}}, {targetBone {boneName=spine_04}}, {targetBone {boneName=spine_05}}, {targetBone {boneName=clavicle_l}}]
    constraintWeights [1,1,1,1]
   }
   n75 = "Constraint - default_backFlap_b_base (PPPP)"(ComponentPose=@n74, Alpha=1, ConstraintWeights_0=0.05, ConstraintWeights_1=0.3, ConstraintWeights_2=0.6, ConstraintWeights_3=0.05) {
    alphaBoolBlend {blendOption=Linear}
    boneToModify {boneName=default_backFlap_b_base}
    constraintSetup [{targetBone {boneName=clavicle_r}}, {targetBone {boneName=spine_04}}, {targetBone {boneName=spine_05}}, {targetBone {boneName=clavicle_l}}]
    constraintWeights [1,1,1,1]
   }
   n76 = "Constraint - default_backFlap_c_base (PPPPPPP)"(
    ComponentPose=@n75
    Alpha=1
    ConstraintWeights_0=0.02
    ConstraintWeights_1=0.05
    ConstraintWeights_2=0.15
    ConstraintWeights_3=0.15
    ConstraintWeights_4=0.35
    ConstraintWeights_5=0.26
    ConstraintWeights_6=0.02
   ) {
    alphaBoolBlend {blendOption=Linear}
    boneToModify {boneName=default_backFlap_c_base}
    constraintSetup [
     {targetBone {boneName=clavicle_l}}
     {targetBone {boneName=spine_01}}
     {targetBone {boneName=spine_02}}
     {targetBone {boneName=spine_03}}
     {targetBone {boneName=spine_04}}
     {targetBone {boneName=spine_05}}
     {targetBone {boneName=clavicle_r}}
    ]
    constraintWeights [1,1,1,1,1,1,1]
   }
   n77 = "Constraint - default_backFlap_d_base (PPPPP)"!(
    ComponentPose=@n76
    Alpha=1
    ConstraintWeights_0=0.1
    ConstraintWeights_1:float=0.4
    ConstraintWeights_2:float=0.15
    ConstraintWeights_3:float=0.25
    ConstraintWeights_4:float=0.1
   ) -> (Pose) {
    alphaBoolBlend {blendOption=Linear}
    boneToModify {boneName=default_backFlap_d_base}
    constraintSetup [
     {targetBone {boneName=spine_01}}
     {targetBone {boneName=spine_02}}
     {targetBone {boneName=spine_03}}
     {targetBone {boneName=spine_04}}
     {targetBone {boneName=pelvis}}
    ]
    constraintWeights [1,1,1,1,1]
   }
   n78 = "Constraint - summer_shirts_bottom_fr_base (PPP)"!(ComponentPose=@n77, Alpha=1, ConstraintWeights_0=0.1, ConstraintWeights_1:float=0.1, ConstraintWeights_2:float=0.8) -> (Pose) {
    alphaBoolBlend {blendOption=Linear}
    boneToModify {boneName=summer_shirts_bottom_fr_base}
    constraintSetup [{targetBone {boneName=spine_02}}, {targetBone {boneName=spine_03}}, {targetBone {boneName=spine_04}}]
    constraintWeights [1,1,1]
   }
   n79 = "Constraint - summer_shirts_bottom_bk_base (PPPPPP)"!(
    ComponentPose=@n78
    Alpha=1
    ConstraintWeights_0=0.09
    ConstraintWeights_1:float=0.35
    ConstraintWeights_2:float=0.17
    ConstraintWeights_3:float=0.24
    ConstraintWeights_4:float=0.05
    ConstraintWeights_5:float=0.1
   ) -> (Pose) {
    alphaBoolBlend {blendOption=Linear}
    boneToModify {boneName=summer_shirts_bottom_bk_base}
    constraintSetup [
     {targetBone {boneName=spine_01}}
     {targetBone {boneName=spine_02}}
     {targetBone {boneName=spine_03}}
     {targetBone {boneName=spine_04}}
     {targetBone {boneName=spine_05}}
     {targetBone {boneName=pelvis}}
    ]
    constraintWeights [1,1,1,1,1,1]
   }
   n80 = "Constraint - summer_shirts_bottom_fr_base_l (PPPP)"!(
    ComponentPose=@n79
    Alpha=1
    ConstraintWeights_0=0.05
    ConstraintWeights_1:float=0.1
    ConstraintWeights_2:float=0.08
    ConstraintWeights_3:float=0.77
   ) -> (Pose) {
    alphaBoolBlend {blendOption=Linear}
    boneToModify {boneName=summer_shirts_bottom_fr_base_l}
    constraintSetup [{targetBone {boneName=spine_01}}, {targetBone {boneName=spine_02}}, {targetBone {boneName=spine_03}}, {targetBone {boneName=spine_04}}]
    constraintWeights [1,1,1,1]
   }
   n81 = "Constraint - summer_shirts_side_base_l (PPPPP)"!(
    ComponentPose=@n80
    Alpha=1
    ConstraintWeights_0=0.1
    ConstraintWeights_1:float=0.12
    ConstraintWeights_2:float=0.15
    ConstraintWeights_3:float=0.55
    ConstraintWeights_4:float=0.08
   ) -> (Pose) {
    alphaBoolBlend {blendOption=Linear}
    boneToModify {boneName=summer_shirts_side_base_l}
    constraintSetup [
     {targetBone {boneName=spine_01}}
     {targetBone {boneName=spine_02}}
     {targetBone {boneName=spine_03}}
     {targetBone {boneName=spine_04}}
     {targetBone {boneName=spine_05}}
    ]
    constraintWeights [1,1,1,1,1]
   }
   n82 = "Constraint - summer_shirts_bottom_fr_base_r (PPPP)"!(
    ComponentPose=@n81
    Alpha=1
    ConstraintWeights_0=0.05
    ConstraintWeights_1:float=0.1
    ConstraintWeights_2:float=0.08
    ConstraintWeights_3:float=0.77
   ) -> (Pose) {
    alphaBoolBlend {blendOption=Linear}
    boneToModify {boneName=summer_shirts_bottom_fr_base_r}
    constraintSetup [{targetBone {boneName=spine_01}}, {targetBone {boneName=spine_02}}, {targetBone {boneName=spine_03}}, {targetBone {boneName=spine_04}}]
    constraintWeights [1,1,1,1]
   }
   n83 = "Constraint - summer_shirts_side_base_r (PPPPP)"!(
    ComponentPose=@n82
    Alpha=1
    ConstraintWeights_0=0.1
    ConstraintWeights_1:float=0.12
    ConstraintWeights_2:float=0.15
    ConstraintWeights_3:float=0.55
    ConstraintWeights_4:float=0.08
   ) -> (Pose) {
    alphaBoolBlend {blendOption=Linear}
    boneToModify {boneName=summer_shirts_side_base_r}
    constraintSetup [
     {targetBone {boneName=spine_01}}
     {targetBone {boneName=spine_02}}
     {targetBone {boneName=spine_03}}
     {targetBone {boneName=spine_04}}
     {targetBone {boneName=spine_05}}
    ]
    constraintWeights [1,1,1,1,1]
   }
   n84 = RigidBody(ComponentPose=@n83) {
    bClampLinearTranslationLimitToRefPose=true
    bOverrideWorldGravity=true
    overridePhysicsAsset=/BRCosmetics/Characters/Player/Female/Medium/Bodies/F_MED_Believer_Summer/Meshes/F_MED_Believer_Summer_ShirtsBottom_Physics
    simulationSpace=BaseBoneSpace
    alphaBoolBlend {blendOption=Linear}
    baseBoneRef {boneName=pelvis}
    componentLinearAccScale=<0.35 0.35 0.35>
    componentLinearVelScale=<0.35 0.35 0.35>
    overrideWorldGravity {z=-980}
   }
   n85 = RigidBody(ComponentPose=@n84) {
    bClampLinearTranslationLimitToRefPose=true
    bOverrideWorldGravity=true
    overridePhysicsAsset=/BRCosmetics/Characters/Player/Female/Medium/Bodies/F_MED_Believer_Summer/Meshes/F_MED_Believer_Summer_ShirtsBack_Physics
    simulationSpace=BaseBoneSpace
    alphaBoolBlend {blendOption=Linear}
    baseBoneRef {boneName=spine_01}
    componentLinearAccScale=<0.3 0.3 0.3>
    componentLinearVelScale=<0.3 0.3 0.3>
    overrideWorldGravity {z=-980}
   }
   n86 = RigidBody(ComponentPose=@n85) {
    bClampLinearTranslationLimitToRefPose=true
    bOverrideWorldGravity=true
    overridePhysicsAsset=/BRCosmetics/Characters/Player/Female/Medium/Bodies/F_MED_Believer_Summer/Meshes/F_MED_Believer_Summer_Strap_Physics
    simulationSpace=BaseBoneSpace
    alphaBoolBlend {blendOption=Linear}
    baseBoneRef {boneName=spine_03}
    componentLinearAccScale=<0.35 0.35 0.35>
    componentLinearVelScale=<0.35 0.35 0.35>
    overrideWorldGravity {z=-980}
   }
   Save cached pose 'summerVariantPose'(Pose=Component To Local(ComponentPose=@n86)) {cacheName=summerVariantPose}
  ]}
 EventGraph {nodes [
   Event BlueprintUpdateAnimation()
   > n1 = Sequence!(execute) -> (then_0, then_1, then_2:Exec, then_3:Exec, then_4:Exec, then_5:Exec, then_6:Exec)
   > n2 = Sequence()
   n3 = GetDeltaTransformFromRefPose(self=GetOwnerSkeletalMeshComponent(), BoneName=lowerarm_l, BaseName=upperarm_l)
   > Set elbow_rot_l(elbow_rot_l:double=BreakRotator(InRot=BreakTransform(InTransform=@n3).Rotation).Yaw) -> (Output_Get:double)
   n4 = GetDeltaTransformFromRefPose(self=GetOwnerSkeletalMeshComponent(), BoneName=lowerarm_r, BaseName=upperarm_r)
   > Set elbow_rot_r(elbow_rot_r:double=BreakRotator(InRot=BreakTransform(InTransform=@n4).Rotation).Yaw) -> (Output_Get:double)
   n5 = GetDeltaTransformFromRefPose(self=GetOwnerSkeletalMeshComponent(), BoneName=calf_l, BaseName=thigh_l)
   Set knee_rot_l(execute=@n2.then_2, knee_rot_l:double=BreakRotator(InRot=BreakTransform(InTransform=@n5).Rotation).Yaw) -> (Output_Get:double)
   n6 = GetDeltaTransformFromRefPose(self=GetOwnerSkeletalMeshComponent(), BoneName=calf_r, BaseName=thigh_r)
   > Set knee_rot_r(knee_rot_r:double=BreakRotator(InRot=BreakTransform(InTransform=@n6).Rotation).Yaw) -> (Output_Get:double)
   n7 = GetDeltaTransformFromRefPose(self=GetOwnerSkeletalMeshComponent(), BoneName=thigh_l, BaseName=pelvis)
   Set thigh_rot_l(execute=@n2.then_1, thigh_rot_l=BreakRotator(InRot=BreakTransform(InTransform=@n7).Rotation).Yaw)
   n8 = GetDeltaTransformFromRefPose(self=GetOwnerSkeletalMeshComponent(), BoneName=thigh_r, BaseName=pelvis)
   > Set thigh_rot_r(thigh_rot_r=BreakRotator(InRot=BreakTransform(InTransform=@n8).Rotation).Yaw)
   n9 = Sequence(execute=@n1.then_1)
   > n10 = PSD Func!(execute, self, startJnt:String=upperarm_l, driverJnt:String=lowerarm_l, poseJnt:String=pose_arm_up_l, angle:double=70, debug:bool=false) -> (then, alpha:double)
   > Set psd_arm_up_alpha_l(psd_arm_up_alpha_l:double=@n10) -> (Output_Get:double)
   > n11 = PSD Func(startJnt=upperarm_r, driverJnt=lowerarm_r, poseJnt=pose_arm_up_r, angle=70, debug=false)
   > Set psd_arm_up_alpha_r(psd_arm_up_alpha_r:double=@n11) -> (Output_Get:double)
   n12 = PSD Func(execute=@n9.then_1, startJnt=upperarm_l, driverJnt=lowerarm_l, poseJnt=pose_arm_fwd_l, angle=50, debug=false)
   > Set psd_arm_fwd_alpha_l(psd_arm_fwd_alpha_l:double=@n12) -> (Output_Get:double)
   > n13 = PSD Func(startJnt=upperarm_r, driverJnt=lowerarm_r, poseJnt=pose_arm_fwd_r, angle=50, debug=false)
   > Set psd_arm_fwd_alpha_r(psd_arm_fwd_alpha_r:double=@n13) -> (Output_Get:double)
   n14 = PSD Func(execute=@n9.then_2, startJnt=upperarm_l, driverJnt=lowerarm_l, poseJnt=pose_arm_bk_l, angle=50, debug=false)
   > Set psd_arm_bk_alpha_l(psd_arm_bk_alpha_l:double=@n14) -> (Output_Get:double)
   > n15 = PSD Func(startJnt=upperarm_r, driverJnt=lowerarm_r, poseJnt=pose_arm_bk_r, angle=50, debug=false)
   > Set psd_arm_bk_alpha_r(psd_arm_bk_alpha_r:double=@n15) -> (Output_Get:double)
   n16 = GetOwnerAnimationInterface!(self:Custom Character Part Anim Instance Object Reference) -> (ReturnValue:Fort Player Animation Interface Object Reference)
   n17 = Set isPistol(
    execute=@n9.then_3
    isPistol="Equal (Enum)"(
     A:EFortWeaponCoreAnimation Enum=GetWeaponCoreAnim!(self:Fort Player Animation Interface Object Reference=@n16) -> (ReturnValue:EFortWeaponCoreAnimation Enum)
     B:EFortWeaponCoreAnimation Enum=Pistol
    )
   )
   > n18 = Branch(Condition=@n17)
   > Set pistol_ads_up_alpha_r(pistol_ads_up_alpha_r:double=Get psd_arm_fwd_alpha_r()) -> (Output_Get:double)
   Set pistol_ads_up_alpha_r(execute=@n18.else, pistol_ads_up_alpha_r=0)
   Event OnCharacterCustomizationLoaded()
   > n19 = Cast To FortPlayerPawn(Object=GetOwningActor())
   > n20 = GetVariantChannelActiveIndexForItemActor!(
    execute
    Item:Athena Cosmetic Item Definition Object Reference=/BRCosmetics/Athena/Items/Cosmetics/Characters/CID_A_114_Athena_Commando_F_Believer
    Actor:Actor Object Reference=@n19
    Channel:Gameplay Tag Structure=(TagName="Cosmetics.Variant.Channel.Progressive")
    VariantList:Array of Gameplay Tag Structures=Make Array!(
     "[0]"=(TagName="Cosmetics.Variant.Property.Stage1")
     "[1]":Gameplay Tag Structure=(TagName="Cosmetics.Variant.Property.Stage2")
     "[2]":Gameplay Tag Structure=(TagName="Cosmetics.Variant.Property.Stage3")
    ) -> (Array)
    bErrorIfChannelMissing:bool=false
    bErrorIfTagNotInArray:bool=false
   ) -> (then, Index:int=0, bIndexIsValid:bool, bFoundChannel:bool)
   > n21 = Branch(Condition=@n20.bIndexIsValid)
   > Set variantIndex(variantIndex:int=@n20) -> (Output_Get:int=0)
   Set variantIndex(execute=@n21.else, variantIndex=0) -> (Output_Get=0)
  ]}
 PSD Func {nodes [
   n1 = PSD Func!() -> (then, startJnt:String, driverJnt:String, poseJnt:String, angle:double=~, debug:bool=~) {is=title K2Node_FunctionEntry}
   > n2 = Sequence!(execute) -> (then_0, then_1)
   > Branch(Condition=@n1.debug)
   n3 = GetPartSkeletalMeshComponent()
   n4 = GetSocketLocation!(self:Scene Component Object Reference=@n3, InSocketName:Name=String To Name(InString=@n1)) -> (ReturnValue:Vector)
   > DrawDebugArrow!(
    execute
    LineStart:Vector=@n4
    LineEnd:Vector=GetSocketLocation(self=@n3, InSocketName=String To Name(InString=@n1.poseJnt))
    ArrowSize:float=0
    LineColor:Linear Color Structure=(R=0.1692,G=0.147,B=1,A=1)
    Duration:float=0
    Thickness:float=1
    DepthPriority:EDrawDebugSceneDepthPriorityGroup Enum=World
   ) -> (then)
   n5 = GetPartSkeletalMeshComponent()
   n6 = Get Direction Between Sockets!(
    Component:Skeletal Mesh Component Object Reference=@n5
    SocketOrBoneNameFrom:Name=String To Name(InString=@n1)
    SocketOrBoneNameTo:Name=String To Name(InString=@n1.poseJnt)
   ) -> (ReturnValue:Vector) {is=title K2Node_CallFunction}
   > Draw Debug Cone!(
    execute
    Origin:Vector=@n4
    Direction:Vector=@n6
    Length:float=100
    AngleWidth:float=@n1.angle
    AngleHeight:float=@n1.angle
    NumSides:int=12
    LineColor:Linear Color Structure=(R=1,G=1,B=1,A=1)
    Duration:float=0
    Thickness:float=0.3
    DepthPriority:EDrawDebugSceneDepthPriorityGroup Enum=World
   ) -> (then)
   n7 = Get Direction Between Sockets(Component=@n5, SocketOrBoneNameFrom=String To Name(InString=@n1), SocketOrBoneNameTo=String To Name(InString=@n1.driverJnt))
   n8 = float - float(B=float / float(A=GetAngleBetweenTwoVectors!(VectorA:Vector=@n7, VectorB:Vector=@n6) -> ("Angle (Degrees)":double), B=@n1.angle) {is=title K2Node_CallFunction}) {is=title K2Node_CallFunction}
   n9 = Branch(execute=@n2.then_1, Condition="float >= float"(A=@n8, B=0) {is=title K2Node_CallFunction})
   > n10 = Set psd_alpha(psd_alpha:double=@n8) -> (Output_Get:double)
   Return Node!(execute=[@n11.then, @n10.then], alpha:double=@n10)
   n11 = Set psd_alpha(execute=@n9.else, psd_alpha=0)
  ]}
]
variables [
 elbow_rot_l {type=number}
 elbow_rot_r {type=number}
 isPistol {type=boolean}
 knee_rot_l {type=number}
 knee_rot_r {type=number}
 pistol_ads_up_alpha_r {type=number}
 psd_alpha {type=number}
 psd_arm_bk_alpha_l {type=number}
 psd_arm_bk_alpha_r {type=number}
 psd_arm_fwd_alpha_l {type=number}
 psd_arm_fwd_alpha_r {type=number}
 psd_arm_up_alpha_l {type=number}
 psd_arm_up_alpha_r {type=number}
 thigh_rot_l {type=number}
 thigh_rot_r {type=number}
 variantIndex {type=integer}
]
